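# Physics asset for the dynamic necklace on the F_MED_Scribble female medium cosmetic body.
# A three-link necklace chain, each link built from two parallel capsules, hangs from a small master body with tight swing and twist limits between links.
# The master links to spine_05, which links to spine_04, both with free angular motion.
# Large capsules on those two spine bones give the upper torso a collision shape.
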PhysicsAsset /BRCosmetics/Characters/Player/Female/Medium/Bodies/F_MED_Scribble/Meshes/F_MED_Scribble_Dyn_necklace
bodies {
 dyn_necklace_01 [
  {shapeType=Capsule, length=7.5179, radius=1.8621, shapeName=dyn_necklace_01_capsule, center=<4.0825 3 0>, rotation=<rot 90 0 0>}
  {shapeType=Capsule, length=7.5179, radius=1.8621, shapeName=dyn_necklace_01_capsule, center=<4.0706 -3 0>, rotation=<rot 90 0 0>}
 ]
 dyn_necklace_01_MSTR [{shapeType=Capsule, length=0.505, radius=0.505, shapeName=dyn_necklace_01_MSTR_capsule, rotation=<rot 0 -90 0>}]
 dyn_necklace_02 [
  {shapeType=Capsule, length=7.5179, radius=1.8621, shapeName=dyn_necklace_01_capsule, center=<4.0825 1.5 0>, rotation=<rot 90 0 0>}
  {shapeType=Capsule, length=7.5179, radius=1.8621, shapeName=dyn_necklace_01_capsule, center=<4.0706 -1.5 0>, rotation=<rot 90 0 0>}
 ]
 dyn_necklace_03 [
  {shapeType=Capsule, length=2.6415, radius=1.8621, shapeName=dyn_necklace_01_capsule, center=<2.2383 1.5 0>, rotation=<rot 90 0 0>}
  {shapeType=Capsule, length=2.6415, radius=1.8621, shapeName=dyn_necklace_01_capsule, center=<2.2264 -1.5 0>, rotation=<rot 90 0 0>}
 ]
 spine_04 [{shapeType=Capsule, length=2.6388, radius=10.8752, shapeName=spine_04_capsule, center=<-0.0035 -1.0407 0>}]
 spine_05 [{shapeType=Capsule, length=10.4191, radius=10.8752, shapeName=spine_04_capsule, center=<-2.3577 -3.3823 0>}]
}
constraints [
 {bone1Name=dyn_necklace_01, bone2Name=dyn_necklace_01_MSTR, swing1LimitDegrees=45, swing2LimitDegrees=25, twistLimitDegrees=5}
 {bone1Name=dyn_necklace_02, bone2Name=dyn_necklace_01, swing1LimitDegrees=15, swing2LimitDegrees=25, twistLimitDegrees=5}
 {bone1Name=dyn_necklace_03, bone2Name=dyn_necklace_02, swing1LimitDegrees=25, swing2LimitDegrees=25, twistLimitDegrees=5}
 {
  bone1Name=dyn_necklace_01_MSTR
  bone2Name=spine_05
  swing1LimitDegrees=45
  swing1Motion=Free
  swing2LimitDegrees=45
  swing2Motion=Free
  twistLimitDegrees=45
  twistMotion=Free
 }
 {
  bone1Name=spine_05
  bone2Name=spine_04
  swing1LimitDegrees=45
  swing1Motion=Free
  swing2LimitDegrees=45
  swing2Motion=Free
  twistLimitDegrees=45
  twistMotion=Free
 }
]
