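# Physics asset for the dynamic skirt on the F_MED_Scribble female player body.
# Front, side and back bone chains of capsules and boxes use mostly limited swing and twist, several hanging from tiny MSTR anchors constrained to pelvis or thighs.
# Pelvis and leg bodies provide collision shapes, and a huge root sphere approximates a floor.
# A cross-constraint links two left panels; a leftover grenade constraint references bones without bodies.
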PhysicsAsset /BRCosmetics/Characters/Player/Female/Medium/Bodies/F_MED_Scribble/Meshes/F_MED_Scribble_Dyn_skirt
bodies {
 calf_l [{
   shapeType=Capsule
   length=51.2296
   radius=9.5854
   shapeName=thigh_r_capsule
   center=<-28.5123 -3.9856 0.9334>
   rotation=<rot 89.1295 -176.1486 2.3844>
  }]
 calf_r [
  {
   shapeType=Capsule
   length=47.2296
   radius=8.0854
   shapeName=thigh_r_capsule
   center=<26.4008 2.3756 -0.9671>
   rotation=<rot 89.1295 -176.1486 -177.6156>
  }
  {shapeType=Capsule, length=10, radius=15, shapeName=calf_r_capsule, center=<42.461 -4.3149 0>, rotation=<rot -90 -99.2265 101.7524>}
 ]
 dyn_skirt_bk_01 [{shapeType=Capsule, length=41.282, radius=6.8547, shapeName=dyn_grenade_01_capsule, center=<4.602 0 5.6419>, rotation=<rot 0 0 90>}]
 dyn_skirt_bk_01_MSTR [{shapeType=Capsule, length=0.505, radius=0.505, shapeName=dyn_skirt_bk_01_MSTR_capsule, rotation=<rot 0 -90 0>}]
 dyn_skirt_bk_02 [{shapeType=Capsule, length=41.282, radius=6.8547, shapeName=dyn_grenade_01_capsule, center=<4.602 0 5.6419>, rotation=<rot 0 0 90>}]
 dyn_skirt_bk_04 [{shapeType=Capsule, length=41.282, radius=6.8547, shapeName=dyn_grenade_01_capsule, center=<4.602 0 4.2673>, rotation=<rot 0 0 90>}]
 dyn_skirt_bk_05 [{shapeType=Capsule, length=41.282, radius=6.8547, shapeName=dyn_grenade_01_capsule, center=<4.602 0 3.6>, rotation=<rot 0 0 90>}]
 dyn_skirt_bk_06 [{shapeType=Capsule, length=41.282, radius=6.8547, shapeName=dyn_grenade_01_capsule, center=<4.602 0 3.6>, rotation=<rot 0 0 90>}]
 dyn_skirt_bk_l_01 [{shapeType=Capsule, length=22.1808, radius=6.8547, shapeName=dyn_grenade_01_capsule, center=<4.602 0 6.3>, rotation=<rot 0 0 90>}]
 dyn_skirt_bk_l_01_MSTR [{shapeType=Capsule, length=0.505, radius=0.505, shapeName=dyn_skirt_bk_l_01_MSTR_capsule, rotation=<rot 0 -90 0>}]
 dyn_skirt_bk_l_02 [{shapeType=Capsule, length=22.1808, radius=6.8547, shapeName=dyn_grenade_01_capsule, center=<4.602 0 6.3>, rotation=<rot 0 0 90>}]
 dyn_skirt_bk_l_04 [{shapeType=Capsule, length=22.1808, radius=6.8547, shapeName=dyn_grenade_01_capsule, center=<4.602 0 6.3>, rotation=<rot 0 0 90>}]
 dyn_skirt_bk_l_05 [{shapeType=Capsule, length=20.2268, radius=6.8547, shapeName=dyn_grenade_01_capsule, center=<4.602 0 6.3>, rotation=<rot 0 0 90>}]
 dyn_skirt_bk_l_06 [{shapeType=Capsule, length=34.954, radius=6.8547, shapeName=dyn_grenade_01_capsule, center=<4.602 0 6.3>, rotation=<rot 0 0 90>}]
 dyn_skirt_bk_r_02 [{shapeType=Capsule, length=40.6291, radius=6.5848, shapeName=dyn_grenade_01_capsule, center=<4.6 0 7.4>, rotation=<rot 0 0 90>}]
 dyn_skirt_bk_r_02_MSTR [{shapeType=Capsule, length=0.505, radius=0.505, shapeName=dyn_skirt_fr_l_01_MSTR_capsule, rotation=<rot 0 -90 0>}]
 dyn_skirt_bk_r_04 [{shapeType=Box, extentX=16.5, extentY=38.5, extentZ=5.5, shapeName=dyn_skirt_fr_l_04_box, center=<6.9694 0 3.4>}]
 dyn_skirt_bk_r_05 [{
   shapeType=Box
   extentX=16.5
   extentY=38.5
   extentZ=5.5
   shapeName=dyn_skirt_fr_l_04_box
   center=<6.9694 0 0>
   rotation=<rot -0.0001 0 -11.4329>
  }]
 dyn_skirt_bk_r_06 [{
   shapeType=Box
   extentX=16.5
   extentY=38.5
   extentZ=5.5
   shapeName=dyn_skirt_fr_l_04_box
   center=<6.9694 0 0>
   rotation=<rot -0.0001 0 -11.4329>
  }]
 dyn_skirt_fr_l_01 [{shapeType=Capsule, length=0.505, radius=0.505, shapeName=dyn_skirt_fr_l_01_MSTR_capsule, rotation=<rot 0 -90 0>}]
 dyn_skirt_fr_l_01_MSTR [{shapeType=Capsule, length=0.505, radius=0.505, shapeName=dyn_skirt_fr_l_01_MSTR_capsule, rotation=<rot 0 -90 0>}]
 dyn_skirt_fr_l_02 [{shapeType=Capsule, length=14.3423, radius=6.248, shapeName=dyn_grenade_01_capsule, center=<9 0 5.7>, rotation=<rot 0 0 90>}]
 dyn_skirt_fr_l_04 [{
   shapeType=Box
   extentX=13.5
   extentY=15.5
   extentZ=1.5
   shapeName=dyn_skirt_fr_l_04_box
   center=<6.9694 7.2554 -0.6008>
   rotation=<rot 0 0 4.2449>
  }]
 dyn_skirt_fr_l_05 [{
   shapeType=Box
   extentX=13.5
   extentY=18.5
   extentZ=1.5
   shapeName=dyn_skirt_fr_l_04_box
   center=<7.2362 7.2554 -0.6008>
   rotation=<rot 0 0 4.2449>
  }]
 dyn_skirt_fr_l_06 [{
   shapeType=Box
   extentX=13.5
   extentY=18.5
   extentZ=1.5
   shapeName=dyn_skirt_fr_l_04_box
   center=<6.9694 7.2554 -0.6008>
   rotation=<rot 0 0 4.2449>
  }]
 dyn_skirt_fr_r_02_MSTR [{shapeType=Capsule, length=0.505, radius=0.505, shapeName=dyn_skirt_fr_l_01_MSTR_capsule, rotation=<rot 0 -90 0>}]
 dyn_skirt_fr_r_04 [{shapeType=Capsule, length=10.974, radius=8, shapeName=dyn_grenade_01_capsule, center=<10 0 8>, rotation=<rot 0 0 90>}]
 dyn_skirt_fr_r_05 [{shapeType=Capsule, length=10.974, radius=8, shapeName=dyn_grenade_01_capsule, center=<10 0 8>, rotation=<rot 0 0 90>}]
 dyn_skirt_fr_r_06 [{shapeType=Capsule, length=10.974, radius=8, shapeName=dyn_grenade_01_capsule, center=<2.8851 0 8>, rotation=<rot 0 0 90>}]
 dyn_skirt_l_01 [{shapeType=Capsule, length=0.505, radius=0.505, shapeName=dyn_skirt_l_01_MSTR_capsule, rotation=<rot 0 -90 0>}]
 dyn_skirt_l_01_MSTR [{shapeType=Capsule, length=0.505, radius=0.505, shapeName=dyn_skirt_l_01_MSTR_capsule, rotation=<rot 0 -90 0>}]
 dyn_skirt_l_02 [{shapeType=Capsule, length=22.1808, radius=6.8547, shapeName=dyn_grenade_01_capsule, center=<4.602 0 6.8>, rotation=<rot 0 0 90>}]
 dyn_skirt_l_04 [{shapeType=Capsule, length=22.1808, radius=6.8547, shapeName=dyn_grenade_01_capsule, center=<4.602 0 6.8>, rotation=<rot 0 0 90>}]
 dyn_skirt_l_05 [{shapeType=Capsule, length=20.6002, radius=8.4353, shapeName=dyn_grenade_01_capsule, center=<4.602 0 6.8>, rotation=<rot 0 0 90>}]
 dyn_skirt_l_06 [{shapeType=Capsule, length=20.6002, radius=8.4353, shapeName=dyn_grenade_01_capsule, center=<4.602 0 6.8>, rotation=<rot 0 0 90>}]
 dyn_skirt_r_02 [{shapeType=Capsule, length=34.7575, radius=6.8547, shapeName=dyn_grenade_01_capsule, center=<4.602 0 6.6>, rotation=<rot 0 0 90>}]
 dyn_skirt_r_02_MSTR [{shapeType=Capsule, length=0.505, radius=0.505, shapeName=dyn_skirt_fr_l_01_MSTR_capsule, rotation=<rot 0 -90 0>}]
 dyn_skirt_r_04 [{shapeType=Capsule, length=34.7575, radius=6.8547, shapeName=dyn_grenade_01_capsule, center=<4.602 0 6.6>, rotation=<rot 0 0 90>}]
 dyn_skirt_r_05 [{shapeType=Capsule, length=53.2673, radius=8.3547, shapeName=dyn_grenade_01_capsule, center=<4.602 0 6.6>, rotation=<rot 0 0 90>}]
 dyn_skirt_r_06 [{shapeType=Capsule, length=52.2673, radius=9.3547, shapeName=dyn_grenade_01_capsule, center=<12.3345 0 6.6>, rotation=<rot 0 0 90>}]
 foot_r [{shapeType=Capsule, length=13.3834, radius=5.7237, shapeName=foot_r_capsule, center=<3.9882 -5.3614 0>, rotation=<rot 90 180 85.0001>}]
 pelvis [{shapeType=Capsule, length=5.1163, radius=10.9706, shapeName=thigh_r_capsule, center=<-3.476 -0.8105 0>}]
 root [{shapeType=Sphere, radius=5000, center=<0 0 -4995>}]
 thigh_l [{
   shapeType=Capsule
   length=27.961
   radius=10.4684
   shapeName=thigh_r_capsule
   center=<-22.5978 -1.1498 4.2402>
   rotation=<rot 86.6538 179.5704 -0.6395>
  }]
 thigh_r [{
   shapeType=Capsule
   length=33.7743
   radius=10.4684
   shapeName=thigh_r_capsule
   center=<22.8165 1.1506 -4.2274>
   rotation=<rot 86.6538 179.5704 179.3605>
  }]
}
constraints [
 {bone1Name=dyn_skirt_fr_l_02, bone2Name=dyn_skirt_fr_l_01, swing1LimitDegrees=5, swing2LimitDegrees=45, twistLimitDegrees=5}
 {bone1Name=dyn_skirt_fr_l_04, bone2Name=dyn_skirt_fr_l_02, swing1LimitDegrees=5, swing2LimitDegrees=20, twistLimitDegrees=5}
 {bone1Name=dyn_skirt_fr_l_05, bone2Name=dyn_skirt_fr_l_04, swing1LimitDegrees=5, swing2LimitDegrees=20, twistLimitDegrees=5}
 {bone1Name=dyn_skirt_l_02, bone2Name=dyn_skirt_l_01, swing1LimitDegrees=30, swing2LimitDegrees=30, twistLimitDegrees=5}
 {bone1Name=dyn_skirt_l_04, bone2Name=dyn_skirt_l_02, swing1LimitDegrees=30, swing2LimitDegrees=30, twistLimitDegrees=5}
 {bone1Name=dyn_skirt_l_05, bone2Name=dyn_skirt_l_04, swing1LimitDegrees=30, swing2LimitDegrees=30, twistLimitDegrees=5}
 {bone1Name=dyn_skirt_l_06, bone2Name=dyn_skirt_l_05, swing1LimitDegrees=30, swing2LimitDegrees=30, twistLimitDegrees=5}
 {
  bone1Name=dyn_skirt_fr_l_01_MSTR
  bone2Name=thigh_l
  swing1LimitDegrees=45
  swing1Motion=Free
  swing2LimitDegrees=45
  swing2Motion=Free
  twistLimitDegrees=45
  twistMotion=Free
 }
 {
  bone1Name=dyn_skirt_l_01_MSTR
  bone2Name=thigh_l
  swing1LimitDegrees=45
  swing1Motion=Free
  swing2LimitDegrees=45
  swing2Motion=Free
  twistLimitDegrees=45
  twistMotion=Free
 }
 {bone1Name=dyn_grenade_02, bone2Name=dyn_grenade_02_MSTR, swing1LimitDegrees=30, swing2LimitDegrees=30, twistLimitDegrees=5}
 {
  bone1Name=dyn_skirt_bk_01_MSTR
  bone2Name=pelvis
  swing1LimitDegrees=45
  swing1Motion=Free
  swing2LimitDegrees=45
  swing2Motion=Free
  twistLimitDegrees=45
  twistMotion=Free
 }
 {bone1Name=dyn_skirt_bk_02, bone2Name=dyn_skirt_bk_01, swing1LimitDegrees=5, swing2LimitDegrees=50, twistLimitDegrees=5}
 {bone1Name=dyn_skirt_bk_04, bone2Name=dyn_skirt_bk_02, swing1LimitDegrees=5, swing2LimitDegrees=50, twistLimitDegrees=5}
 {bone1Name=dyn_skirt_bk_05, bone2Name=dyn_skirt_bk_04, swing1LimitDegrees=5, swing2LimitDegrees=50, twistLimitDegrees=5}
 {bone1Name=dyn_skirt_bk_06, bone2Name=dyn_skirt_bk_05, swing1LimitDegrees=5, swing2LimitDegrees=50, twistLimitDegrees=5}
 {
  bone1Name=dyn_skirt_bk_l_01_MSTR
  bone2Name=pelvis
  swing1LimitDegrees=45
  swing1Motion=Free
  swing2LimitDegrees=45
  swing2Motion=Free
  twistLimitDegrees=45
  twistMotion=Free
 }
 {bone1Name=dyn_skirt_bk_l_02, bone2Name=dyn_skirt_bk_l_01, swing1LimitDegrees=30, swing2LimitDegrees=30, twistLimitDegrees=5}
 {bone1Name=dyn_skirt_bk_l_04, bone2Name=dyn_skirt_bk_l_02, swing1LimitDegrees=15, swing2LimitDegrees=15, twistLimitDegrees=5}
 {bone1Name=dyn_skirt_bk_l_05, bone2Name=dyn_skirt_bk_l_04, swing1LimitDegrees=15, swing2LimitDegrees=15, twistLimitDegrees=5}
 {bone1Name=dyn_skirt_bk_l_06, bone2Name=dyn_skirt_bk_l_05, swing1LimitDegrees=15, swing2LimitDegrees=15, twistLimitDegrees=5}
 {
  bone1Name=dyn_skirt_bk_r_02_MSTR
  bone2Name=thigh_r
  swing1LimitDegrees=45
  swing1Motion=Free
  swing2LimitDegrees=45
  swing2Motion=Free
  twistLimitDegrees=45
  twistMotion=Free
 }
 {
  bone1Name=dyn_skirt_bk_r_04
  bone2Name=dyn_skirt_bk_r_02
  swing1LimitDegrees=5
  swing1Motion=Locked
  swing2LimitDegrees=60
  twistLimitDegrees=5
 }
 {
  bone1Name=dyn_skirt_bk_r_05
  bone2Name=dyn_skirt_bk_r_04
  swing1LimitDegrees=5
  swing1Motion=Locked
  swing2LimitDegrees=60
  twistLimitDegrees=5
 }
 {
  bone1Name=dyn_skirt_bk_r_06
  bone2Name=dyn_skirt_bk_r_05
  swing1LimitDegrees=5
  swing1Motion=Locked
  swing2LimitDegrees=60
  twistLimitDegrees=5
 }
 {bone1Name=dyn_skirt_fr_r_04, bone2Name=dyn_skirt_fr_r_02_MSTR, swing1LimitDegrees=2, swing2LimitDegrees=15, twistLimitDegrees=2}
 {bone1Name=dyn_skirt_fr_r_05, bone2Name=dyn_skirt_fr_r_04, swing1LimitDegrees=2, swing2LimitDegrees=60, twistLimitDegrees=2}
 {bone1Name=dyn_skirt_fr_r_06, bone2Name=dyn_skirt_fr_r_05, swing1LimitDegrees=2, swing2LimitDegrees=60, twistLimitDegrees=2}
 {bone1Name=dyn_skirt_r_04, bone2Name=dyn_skirt_r_02, swing1LimitDegrees=15, swing2LimitDegrees=45, twistLimitDegrees=5}
 {bone1Name=dyn_skirt_r_05, bone2Name=dyn_skirt_r_04, swing1LimitDegrees=15, swing2LimitDegrees=45, twistLimitDegrees=15}
 {bone1Name=dyn_skirt_r_06, bone2Name=dyn_skirt_r_05, swing1LimitDegrees=15, swing2LimitDegrees=45, twistLimitDegrees=15}
 {
  bone1Name=thigh_l
  bone2Name=pelvis
  swing1LimitDegrees=45
  swing1Motion=Free
  swing2LimitDegrees=45
  swing2Motion=Free
  twistLimitDegrees=45
  twistMotion=Free
 }
 {
  bone1Name=thigh_r
  bone2Name=pelvis
  swing1LimitDegrees=45
  swing1Motion=Free
  swing2LimitDegrees=45
  swing2Motion=Free
  twistLimitDegrees=45
  twistMotion=Free
 }
 {
  bone1Name=calf_r
  bone2Name=thigh_r
  swing1LimitDegrees=45
  swing1Motion=Free
  swing2LimitDegrees=45
  swing2Motion=Free
  twistLimitDegrees=45
  twistMotion=Free
 }
 {
  bone1Name=calf_l
  bone2Name=thigh_l
  swing1LimitDegrees=45
  swing1Motion=Free
  swing2LimitDegrees=45
  swing2Motion=Free
  twistLimitDegrees=45
  twistMotion=Free
 }
 {
  bone1Name=pelvis
  bone2Name=root
  swing1LimitDegrees=45
  swing1Motion=Free
  swing2LimitDegrees=45
  swing2Motion=Free
  twistLimitDegrees=45
  twistMotion=Free
 }
 {bone1Name=dyn_skirt_r_02, bone2Name=dyn_skirt_r_02_MSTR, swing1LimitDegrees=60, swing2LimitDegrees=60, twistLimitDegrees=5}
 {bone1Name=dyn_skirt_bk_01, bone2Name=dyn_skirt_bk_01_MSTR, swing1LimitDegrees=5, swing2LimitDegrees=50, twistLimitDegrees=5}
 {bone1Name=dyn_skirt_bk_l_01, bone2Name=dyn_skirt_bk_l_01_MSTR, swing1LimitDegrees=30, swing2LimitDegrees=30, twistLimitDegrees=5}
 {bone1Name=dyn_skirt_bk_r_02, bone2Name=dyn_skirt_bk_r_02_MSTR, swing1LimitDegrees=5, swing2LimitDegrees=60, twistLimitDegrees=5}
 {
  bone1Name=dyn_skirt_fr_l_01
  bone2Name=dyn_skirt_fr_l_01_MSTR
  swing1LimitDegrees=45
  swing1Motion=Free
  swing2LimitDegrees=45
  swing2Motion=Free
  twistLimitDegrees=45
  twistMotion=Free
 }
 {
  bone1Name=dyn_skirt_l_01
  bone2Name=dyn_skirt_l_01_MSTR
  swing1LimitDegrees=45
  swing1Motion=Free
  swing2LimitDegrees=45
  swing2Motion=Free
  twistLimitDegrees=45
  twistMotion=Free
 }
 {
  bone1Name=foot_r
  bone2Name=calf_r
  swing1LimitDegrees=45
  swing1Motion=Free
  swing2LimitDegrees=45
  swing2Motion=Free
  twistLimitDegrees=45
  twistMotion=Free
 }
 {bone1Name=dyn_skirt_fr_l_06, bone2Name=dyn_skirt_fr_l_05, swing1LimitDegrees=5, swing2LimitDegrees=20, twistLimitDegrees=5}
 {
  bone1Name=dyn_skirt_l_05
  bone2Name=dyn_skirt_bk_l_05
  swing1LimitDegrees=45
  swing1Motion=Free
  swing2LimitDegrees=45
  swing2Motion=Free
  twistLimitDegrees=45
  twistMotion=Free
 }
]
