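# Post-pose AnimBP for the F_MED_Scribble player body, copying the attached parent's pose and adding secondary motion.
# A deform Control Rig feeds a second Control Rig that receives knee-slide, baseball-slide and crouch flags.
# RigidBody nodes then simulate necklace and skirt, with in-vehicle alphas controlling skirt simulation and shrinking selected skirt bones.
# A long chain of per-bone AnimDynamics springs moves grenades, their pins and arm tassels.
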AnimBlueprint /BRCosmetics/Characters/Player/Female/Medium/Bodies/F_MED_Scribble/Meshes/F_MED_Scribble_AnimBP
target_skeleton=&/F_MED_Scribble_Skeleton
graphs [
 AnimGraph {nodes [
   n1 = Control Rig(Source=Copy Pose From Mesh() {bUseAttachedParent=true}) {
    alphaBoolBlend {blendOption=Linear}
    controlRigAssetReference {blueprintRigClass=&/F_MED_Scribble_DeformRig.F_MED_Scribble_DeformRig_C}
   }
   n2 = Control Rig!(
    Source=@n1
    Alpha
    isKneeSliding:bool=Get bIsKneeSliding()
    isBaseballSliding:bool=Get bIsBaseballSliding()
    isCrouching:bool=Get bIsCrouching()
   ) -> (Pose) {
    alphaBoolBlend {blendOption=Linear}
    controlRigAssetReference {blueprintRigClass=&/F_MED_Scribble_ControlRig.F_MED_Scribble_ControlRig_C}
   }
   n3 = RigidBody!(
    ComponentPose=Local To Component(LocalPose=@n2)
    Alpha
    OverrideWorldGravity:Vector="0,0,-980"
    ComponentLinearAccScale:Vector="0,0,0"
    ComponentLinearVelScale:Vector="0,0,0"
    SimSpaceSettings:Sim Space Settings Structure=(WorldAlpha=1,VelocityScaleZ=0,DampingAlpha=1,MaxLinearVelocity=500,MaxAngularVelocity=3,MaxLinearAcceleration=250,MaxAngularAcceleration=1000,ExternalLinearDragV=(X=0,Y=0.5,Z=0),ExternalLinearVelocity=(X=0,Y=0,Z=0),ExternalAngularVelocity=(X=0,Y=0,Z=0))
   ) -> (Pose) {
    bOverrideWorldGravity=true
    overridePhysicsAsset=&/F_MED_Scribble_Dyn_necklace
    simulationSpace=BaseBoneSpace
    alphaBoolBlend {blendOption=Linear}
    baseBoneRef {boneName=spine_04}
    componentAppliedLinearAccClamp=<3000 3000 3000>
    overrideWorldGravity {z=-980}
    simSpaceSettings {
     maxAngularAcceleration=1000
     maxAngularVelocity=3
     maxLinearAcceleration=250
     maxLinearVelocity=500
     velocityScaleZ=0
     worldAlpha=1
     externalLinearDragV {y=0.5}
    }
   }
   n4 = RigidBody!(
    ComponentPose=@n3
    Alpha=Get in vehicle rigid alpha() -> (in vehicle rigid alpha:double)
    OverrideWorldGravity="0,0,-980"
    SimSpaceSettings=(WorldAlpha=1,VelocityScaleZ=0,DampingAlpha=1,MaxLinearVelocity=500,MaxAngularVelocity=3,MaxLinearAcceleration=250,MaxAngularAcceleration=1000,ExternalLinearDragV=(X=0,Y=0.5,Z=0),ExternalLinearVelocity=(X=0,Y=0,Z=0),ExternalAngularVelocity=(X=0,Y=0,Z=0))
   ) -> (Pose) {
    bClampLinearTranslationLimitToRefPose=true
    bOverrideWorldGravity=true
    overridePhysicsAsset=&/F_MED_Scribble_Dyn_skirt
    simulationSpace=BaseBoneSpace
    alphaBoolBlend {blendOption=Linear}
    baseBoneRef {boneName=pelvis}
    overrideWorldGravity {z=-980}
    simSpaceSettings {
     maxAngularAcceleration=1000
     maxAngularVelocity=3
     maxLinearAcceleration=250
     maxLinearVelocity=500
     velocityScaleZ=0
     worldAlpha=1
     externalLinearDragV {y=0.5}
    }
   }
   n5 = Get in vehicle scale alpha() -> (in vehicle scale alpha:double)
   n6 = "Transform (Modify) Bone - Bone: dyn_skirt_fr_r_02_MSTR"(ComponentPose=@n4, Alpha=@n5, Translation="0,0,0", Rotation="0,0,0", Scale="0.5,0.8,0.8") {scaleMode=BMM_Replace, alphaBoolBlend {blendOption=Linear}, boneToModify {boneName=dyn_skirt_fr_r_02_MSTR}, scale=<0.5 0.8 0.8>}
   n7 = "Transform (Modify) Bone - Bone: dyn_skirt_r_02_MSTR"(ComponentPose=@n6, Alpha=@n5, Translation="0,0,0", Rotation="0,0,0", Scale="0.5,0.8,0.8") {scaleMode=BMM_Replace, alphaBoolBlend {blendOption=Linear}, boneToModify {boneName=dyn_skirt_r_02_MSTR}, scale=<0.5 0.8 0.8>}
   n8 = "Transform (Modify) Bone - Bone: dyn_skirt_bk_l_01_MSTR"(ComponentPose=@n7, Alpha=@n5, Translation="0,0,0", Rotation="0,0,0", Scale="0.5,0.8,0.8") {scaleMode=BMM_Replace, alphaBoolBlend {blendOption=Linear}, boneToModify {boneName=dyn_skirt_bk_l_01_MSTR}, scale=<0.5 0.8 0.8>}
   n9 = "Transform (Modify) Bone - Bone: dyn_skirt_bk_01_MSTR"(ComponentPose=@n8, Alpha=@n5, Translation="0,0,0", Rotation="0,0,0", Scale="0.5,0.8,0.8") {scaleMode=BMM_Replace, alphaBoolBlend {blendOption=Linear}, boneToModify {boneName=dyn_skirt_bk_01_MSTR}, scale=<0.5 0.8 0.8>}
   n10 = "Transform (Modify) Bone - Bone: dyn_skirt_l_01_MSTR"(ComponentPose=@n9, Alpha=@n5, Translation="0,0,0", Rotation="0,0,0", Scale="0.5,0.8,0.8") {scaleMode=BMM_Replace, alphaBoolBlend {blendOption=Linear}, boneToModify {boneName=dyn_skirt_l_01_MSTR}, scale=<0.5 0.8 0.8>}
   n11 = "Transform (Modify) Bone - Bone: dyn_skirt_fr_l_01_MSTR"(ComponentPose=@n10, Alpha=@n5, Translation="0,0,0", Rotation="0,0,0", Scale="0.5,0.8,0.8") {scaleMode=BMM_Replace, alphaBoolBlend {blendOption=Linear}, boneToModify {boneName=dyn_skirt_fr_l_01_MSTR}, scale=<0.5 0.8 0.8>}
   n12 = "Transform (Modify) Bone - Bone: dyn_skirt_bk_r_02_MSTR"(ComponentPose=@n11, Alpha=@n5, Translation="0,0,0", Rotation="0,0,0", Scale="0.5,0.8,0.8") {scaleMode=BMM_Replace, alphaBoolBlend {blendOption=Linear}, boneToModify {boneName=dyn_skirt_bk_r_02_MSTR}, scale=<0.5 0.8 0.8>}
   n13 = Anim Dynamics - Bone: dyn_grenade_01(ComponentPose=@n12, Alpha=1, ExternalForce="0,0,0") {
    angularDampingOverride=0.9
    angularSpringConstant=2000
    bAngularSpring=true
    bOverrideAngularDamping=true
    bUseGravityOverride=true
    gravityScale=0.35
    simulationSpace=RootRelative
    alphaBoolBlend {blendOption=Linear}
    boundBone {boneName=dyn_grenade_01}
    componentLinearAccScale=<1 1 1>
    componentLinearVelScale=<1 1 1>
    gravityOverride {z=-980}
    physicsBodyDefinitions [{
      boundBone {boneName=dyn_grenade_01}
      boxExtents=<20 30 30>
      constraintSetup {angularLimitsMax=<5 0 20>, angularLimitsMin=<-5 -20 -20>, angularTarget=<1 0 0>}
      localJointOffset=<4 0 0>
     }]
   }
   n14 = Anim Dynamics - Bone: dyn_grenade_02(ComponentPose=@n13, Alpha=1, ExternalForce="0,0,0") {
    angularDampingOverride=0.9
    angularSpringConstant=2000
    bAngularSpring=true
    bOverrideAngularDamping=true
    bUseGravityOverride=true
    gravityScale=0.35
    simulationSpace=RootRelative
    alphaBoolBlend {blendOption=Linear}
    boundBone {boneName=dyn_grenade_02}
    componentLinearAccScale=<1 1 1>
    componentLinearVelScale=<1 1 1>
    gravityOverride {z=-980}
    physicsBodyDefinitions [{
      boundBone {boneName=dyn_grenade_02}
      boxExtents=<20 30 30>
      constraintSetup {angularLimitsMax=<5 0 20>, angularLimitsMin=<-5 -20 -20>, angularTarget=<1 0 0>}
      localJointOffset=<4 0 0>
     }]
    relativeSpaceBone {boneName=pelvis}
   }
   n15 = Anim Dynamics - Bone: dyn_grenade_03(ComponentPose=@n14, Alpha=1, ExternalForce="0,0,0") {
    angularDampingOverride=0.9
    angularSpringConstant=2000
    bAngularSpring=true
    bOverrideAngularDamping=true
    bUseGravityOverride=true
    gravityScale=0.35
    simulationSpace=RootRelative
    alphaBoolBlend {blendOption=Linear}
    boundBone {boneName=dyn_grenade_03}
    componentLinearAccScale=<1 1 1>
    componentLinearVelScale=<1 1 1>
    gravityOverride {z=-980}
    physicsBodyDefinitions [{
      boundBone {boneName=dyn_grenade_03}
      boxExtents=<20 30 30>
      constraintSetup {angularLimitsMax=<5 0 30>, angularLimitsMin=<-5 -15 -30>, angularTarget=<1 0 0>}
      localJointOffset=<4 0 0>
     }]
   }
   n16 = Anim Dynamics - Bone: dyn_grenade_04(ComponentPose=@n15, Alpha=1, ExternalForce="0,0,0") {
    angularDampingOverride=0.9
    angularSpringConstant=2000
    bAngularSpring=true
    bOverrideAngularDamping=true
    bUseGravityOverride=true
    gravityScale=0.35
    simulationSpace=RootRelative
    alphaBoolBlend {blendOption=Linear}
    boundBone {boneName=dyn_grenade_04}
    componentLinearAccScale=<1 1 1>
    componentLinearVelScale=<1 1 1>
    gravityOverride {z=-980}
    physicsBodyDefinitions [{
      boundBone {boneName=dyn_grenade_04}
      boxExtents=<20 30 30>
      constraintSetup {angularLimitsMax=<5 0 30>, angularLimitsMin=<-5 -15 -30>, angularTarget=<1 0 0>}
      localJointOffset=<4 0 0>
     }]
   }
   n17 = Anim Dynamics - Bone: dyn_grenade_01_pin(ComponentPose=@n16, Alpha=1, ExternalForce="0,0,0") {
    angularDampingOverride=1
    angularSpringConstant=3000
    bAngularSpring=true
    bOverrideAngularDamping=true
    bUseGravityOverride=true
    gravityScale=0.35
    simulationSpace=BoneRelative
    alphaBoolBlend {blendOption=Linear}
    boundBone {boneName=dyn_grenade_01_pin}
    componentLinearAccScale=<1 1 1>
    componentLinearVelScale=<1 1 1>
    gravityOverride {z=-980}
    physicsBodyDefinitions [{
      boundBone {boneName=dyn_grenade_01_pin}
      boxExtents=<10 10 3>
      constraintSetup {angularLimitsMax=<15 15 15>, angularLimitsMin=<-15 -15 -15>, angularTarget=<1 0 0>}
      localJointOffset=<3 0 0>
     }]
    relativeSpaceBone {boneName=dyn_grenade_01}
   }
   n18 = Anim Dynamics - Bone: dyn_grenade_02_pin(ComponentPose=@n17, Alpha=1, ExternalForce="0,0,0") {
    angularDampingOverride=1
    angularSpringConstant=3000
    bAngularSpring=true
    bOverrideAngularDamping=true
    bUseGravityOverride=true
    gravityScale=0.35
    simulationSpace=BoneRelative
    alphaBoolBlend {blendOption=Linear}
    boundBone {boneName=dyn_grenade_02_pin}
    componentLinearAccScale=<1 1 1>
    componentLinearVelScale=<1 1 1>
    gravityOverride {z=-980}
    physicsBodyDefinitions [{
      boundBone {boneName=dyn_grenade_02_pin}
      boxExtents=<10 10 3>
      constraintSetup {angularLimitsMax=<15 15 15>, angularLimitsMin=<-15 -15 -15>, angularTarget=<1 0 0>}
      localJointOffset=<3 0 0>
     }]
    relativeSpaceBone {boneName=dyn_grenade_02}
   }
   n19 = Anim Dynamics - Bone: dyn_grenade_03_pin(ComponentPose=@n18, Alpha=1, ExternalForce="0,0,0") {
    angularDampingOverride=1
    angularSpringConstant=3000
    bAngularSpring=true
    bOverrideAngularDamping=true
    bUseGravityOverride=true
    gravityScale=0.35
    simulationSpace=BoneRelative
    alphaBoolBlend {blendOption=Linear}
    boundBone {boneName=dyn_grenade_03_pin}
    componentLinearAccScale=<1 1 1>
    componentLinearVelScale=<1 1 1>
    gravityOverride {z=-980}
    physicsBodyDefinitions [{
      boundBone {boneName=dyn_grenade_03_pin}
      boxExtents=<10 10 3>
      constraintSetup {angularLimitsMax=<15 15 15>, angularLimitsMin=<-15 -15 -15>, angularTarget=<1 0 0>}
      localJointOffset=<3 0 0>
     }]
    relativeSpaceBone {boneName=dyn_grenade_03}
   }
   n20 = Anim Dynamics - Bone: dyn_grenade_04_pin(ComponentPose=@n19, Alpha=1, ExternalForce="0,0,0") {
    angularDampingOverride=1
    angularSpringConstant=3000
    bAngularSpring=true
    bOverrideAngularDamping=true
    bUseGravityOverride=true
    gravityScale=0.35
    simulationSpace=BoneRelative
    alphaBoolBlend {blendOption=Linear}
    boundBone {boneName=dyn_grenade_04_pin}
    componentLinearAccScale=<1 1 1>
    componentLinearVelScale=<1 1 1>
    gravityOverride {z=-980}
    physicsBodyDefinitions [{
      boundBone {boneName=dyn_grenade_04_pin}
      boxExtents=<10 10 3>
      constraintSetup {angularLimitsMax=<15 15 15>, angularLimitsMin=<-15 -15 -15>, angularTarget=<1 0 0>}
      localJointOffset=<3 0 0>
     }]
    relativeSpaceBone {boneName=dyn_grenade_04}
   }
   n21 = Anim Dynamics - Bone: dyn_arm_tassle_l_A_01(ComponentPose=@n20, Alpha=1, ExternalForce="0,0,0") {
    angularDampingOverride=1
    angularSpringConstant=20000
    bAngularSpring=true
    bOverrideAngularDamping=true
    bUseGravityOverride=true
    gravityScale=0.35
    simulationSpace=RootRelative
    alphaBoolBlend {blendOption=Linear}
    boundBone {boneName=dyn_arm_tassle_l_A_01}
    componentAppliedLinearAccClamp=<3000 3000 3000>
    componentLinearAccScale=<10 10 10>
    componentLinearVelScale=<10 10 10>
    gravityOverride {z=-980}
    physicsBodyDefinitions [{
      boundBone {boneName=dyn_arm_tassle_l_A_01}
      boxExtents=<18 26 27>
      constraintSetup {angularLimitsMax=<15 2 15>, angularLimitsMin=<-15 -25 -15>, angularTarget=<1 0 0>}
      localJointOffset=<4 0 0>
     }]
   }
   n22 = Anim Dynamics - Bone: dyn_arm_tassle_l_A_02(ComponentPose=@n21, Alpha=1, ExternalForce="0,0,0") {
    angularDampingOverride=1
    angularSpringConstant=20000
    bAngularSpring=true
    bOverrideAngularDamping=true
    bUseGravityOverride=true
    gravityScale=0.35
    simulationSpace=RootRelative
    alphaBoolBlend {blendOption=Linear}
    boundBone {boneName=dyn_arm_tassle_l_A_02}
    componentAppliedLinearAccClamp=<3000 3000 3000>
    componentLinearAccScale=<10 10 10>
    componentLinearVelScale=<10 10 10>
    gravityOverride {z=-980}
    physicsBodyDefinitions [{
      boundBone {boneName=dyn_arm_tassle_l_A_02}
      boxExtents=<20 30 30>
      constraintSetup {angularLimitsMax=<15 2 15>, angularLimitsMin=<-15 -25 -15>, angularTarget=<1 0 0>}
      localJointOffset=<4 0 0>
     }]
   }
   n23 = Anim Dynamics - Bone: dyn_arm_tassle_l_B_01(ComponentPose=@n22, Alpha=1, ExternalForce="0,0,0") {
    angularDampingOverride=1
    angularSpringConstant=20000
    bAngularSpring=true
    bOverrideAngularDamping=true
    bUseGravityOverride=true
    gravityScale=0.35
    simulationSpace=RootRelative
    alphaBoolBlend {blendOption=Linear}
    boundBone {boneName=dyn_arm_tassle_l_B_01}
    componentAppliedLinearAccClamp=<3000 3000 3000>
    componentLinearAccScale=<10 10 10>
    componentLinearVelScale=<10 10 10>
    gravityOverride {z=-980}
    physicsBodyDefinitions [{
      boundBone {boneName=dyn_arm_tassle_l_B_01}
      boxExtents=<20 30 30>
      constraintSetup {angularLimitsMax=<15 2 15>, angularLimitsMin=<-15 -25 -15>, angularTarget=<1 0 0>}
      localJointOffset=<4 0 0>
     }]
   }
   n24 = Anim Dynamics - Bone: dyn_arm_tassle_l_B_02(ComponentPose=@n23, Alpha=1, ExternalForce="0,0,0") {
    angularDampingOverride=1
    angularSpringConstant=20000
    bAngularSpring=true
    bOverrideAngularDamping=true
    bUseGravityOverride=true
    gravityScale=0.35
    simulationSpace=RootRelative
    alphaBoolBlend {blendOption=Linear}
    boundBone {boneName=dyn_arm_tassle_l_B_02}
    componentAppliedLinearAccClamp=<3000 3000 3000>
    componentLinearAccScale=<10 10 10>
    componentLinearVelScale=<10 10 10>
    gravityOverride {z=-980}
    physicsBodyDefinitions [{
      boundBone {boneName=dyn_arm_tassle_l_B_02}
      boxExtents=<20 30 30>
      constraintSetup {angularLimitsMax=<15 2 15>, angularLimitsMin=<-15 -25 -15>, angularTarget=<1 0 0>}
      localJointOffset=<4 0 0>
     }]
   }
   n25 = Anim Dynamics - Bone: dyn_arm_tassle_l_C_01(ComponentPose=@n24, Alpha=1, ExternalForce="0,0,0") {
    angularDampingOverride=1
    angularSpringConstant=20000
    bAngularSpring=true
    bOverrideAngularDamping=true
    bUseGravityOverride=true
    gravityScale=0.35
    simulationSpace=RootRelative
    alphaBoolBlend {blendOption=Linear}
    boundBone {boneName=dyn_arm_tassle_l_C_01}
    componentAppliedLinearAccClamp=<3000 3000 3000>
    componentLinearAccScale=<10 10 10>
    componentLinearVelScale=<10 10 10>
    gravityOverride {z=-980}
    physicsBodyDefinitions [{
      boundBone {boneName=dyn_arm_tassle_l_C_01}
      boxExtents=<22 29 31>
      constraintSetup {angularLimitsMax=<15 2 15>, angularLimitsMin=<-15 -25 -15>, angularTarget=<1 0 0>}
      localJointOffset=<4 0 0>
     }]
   }
   n26 = Anim Dynamics - Bone: dyn_arm_tassle_l_C_02(ComponentPose=@n25, Alpha=1, ExternalForce="0,0,0") {
    angularDampingOverride=1
    angularSpringConstant=20000
    bAngularSpring=true
    bOverrideAngularDamping=true
    bUseGravityOverride=true
    gravityScale=0.35
    simulationSpace=RootRelative
    alphaBoolBlend {blendOption=Linear}
    boundBone {boneName=dyn_arm_tassle_l_C_02}
    componentAppliedLinearAccClamp=<3000 3000 3000>
    componentLinearAccScale=<10 10 10>
    componentLinearVelScale=<10 10 10>
    gravityOverride {z=-980}
    physicsBodyDefinitions [{
      boundBone {boneName=dyn_arm_tassle_l_C_02}
      boxExtents=<20 30 30>
      constraintSetup {angularLimitsMax=<15 2 15>, angularLimitsMin=<-15 -25 -15>, angularTarget=<1 0 0>}
      localJointOffset=<4 0 0>
     }]
   }
   n27 = Anim Dynamics - Bone: dyn_arm_tassle_l_D_01(ComponentPose=@n26, Alpha=1, ExternalForce="0,0,0") {
    angularDampingOverride=1
    angularSpringConstant=20000
    bAngularSpring=true
    bOverrideAngularDamping=true
    bUseGravityOverride=true
    gravityScale=0.35
    simulationSpace=RootRelative
    alphaBoolBlend {blendOption=Linear}
    boundBone {boneName=dyn_arm_tassle_l_D_01}
    componentAppliedLinearAccClamp=<3000 3000 3000>
    componentLinearAccScale=<10 10 10>
    componentLinearVelScale=<10 10 10>
    gravityOverride {z=-980}
    physicsBodyDefinitions [{
      boundBone {boneName=dyn_arm_tassle_l_D_01}
      boxExtents=<22 29 31>
      constraintSetup {angularLimitsMax=<15 2 15>, angularLimitsMin=<-15 -25 -15>, angularTarget=<1 0 0>}
      localJointOffset=<4 0 0>
     }]
   }
   n28 = Anim Dynamics - Bone: dyn_arm_tassle_l_D_02(ComponentPose=@n27, Alpha=1, ExternalForce="0,0,0") {
    angularDampingOverride=1
    angularSpringConstant=20000
    bAngularSpring=true
    bOverrideAngularDamping=true
    bUseGravityOverride=true
    gravityScale=0.35
    simulationSpace=RootRelative
    alphaBoolBlend {blendOption=Linear}
    boundBone {boneName=dyn_arm_tassle_l_D_02}
    componentAppliedLinearAccClamp=<3000 3000 3000>
    componentLinearAccScale=<10 10 10>
    componentLinearVelScale=<10 10 10>
    gravityOverride {z=-980}
    physicsBodyDefinitions [{
      boundBone {boneName=dyn_arm_tassle_l_D_02}
      boxExtents=<20 30 30>
      constraintSetup {angularLimitsMax=<15 2 15>, angularLimitsMin=<-15 -25 -15>, angularTarget=<1 0 0>}
      localJointOffset=<4 0 0>
     }]
   }
   n29 = Anim Dynamics - Bone: dyn_arm_tassle_l_E_01(ComponentPose=@n28, Alpha=1, ExternalForce="0,0,0") {
    angularDampingOverride=1
    angularSpringConstant=20000
    bAngularSpring=true
    bOverrideAngularDamping=true
    bUseGravityOverride=true
    gravityScale=0.35
    simulationSpace=RootRelative
    alphaBoolBlend {blendOption=Linear}
    boundBone {boneName=dyn_arm_tassle_l_E_01}
    componentAppliedLinearAccClamp=<3000 3000 3000>
    componentLinearAccScale=<10 10 10>
    componentLinearVelScale=<10 10 10>
    gravityOverride {z=-980}
    physicsBodyDefinitions [{
      boundBone {boneName=dyn_arm_tassle_l_E_01}
      boxExtents=<20 30 30>
      constraintSetup {angularLimitsMax=<15 2 15>, angularLimitsMin=<-15 -25 -15>, angularTarget=<1 0 0>}
      localJointOffset=<4 0 0>
     }]
   }
   n30 = Anim Dynamics - Bone: dyn_arm_tassle_l_E_02(ComponentPose=@n29, Alpha=1, ExternalForce="0,0,0") {
    angularDampingOverride=1
    angularSpringConstant=20000
    bAngularSpring=true
    bOverrideAngularDamping=true
    bUseGravityOverride=true
    gravityScale=0.35
    simulationSpace=RootRelative
    alphaBoolBlend {blendOption=Linear}
    boundBone {boneName=dyn_arm_tassle_l_E_02}
    componentAppliedLinearAccClamp=<3000 3000 3000>
    componentLinearAccScale=<10 10 10>
    componentLinearVelScale=<10 10 10>
    gravityOverride {z=-980}
    physicsBodyDefinitions [{
      boundBone {boneName=dyn_arm_tassle_l_E_02}
      boxExtents=<20 30 30>
      constraintSetup {angularLimitsMax=<15 2 15>, angularLimitsMin=<-15 -25 -15>, angularTarget=<1 0 0>}
      localJointOffset=<4 0 0>
     }]
   }
   n31 = Anim Dynamics - Bone: dyn_arm_tassle_l_F_01(ComponentPose=@n30, Alpha=1, ExternalForce="0,0,0") {
    angularDampingOverride=1
    angularSpringConstant=20000
    bAngularSpring=true
    bOverrideAngularDamping=true
    bUseGravityOverride=true
    gravityScale=0.35
    simulationSpace=RootRelative
    alphaBoolBlend {blendOption=Linear}
    boundBone {boneName=dyn_arm_tassle_l_F_01}
    componentAppliedLinearAccClamp=<3000 3000 3000>
    componentLinearAccScale=<10 10 10>
    componentLinearVelScale=<10 10 10>
    gravityOverride {z=-980}
    physicsBodyDefinitions [{
      boundBone {boneName=dyn_arm_tassle_l_F_01}
      boxExtents=<22 29 31>
      constraintSetup {angularLimitsMax=<15 2 15>, angularLimitsMin=<-15 -25 -15>, angularTarget=<1 0 0>}
      localJointOffset=<4 0 0>
     }]
   }
   n32 = Anim Dynamics - Bone: dyn_arm_tassle_l_F_02(ComponentPose=@n31, Alpha=1, ExternalForce="0,0,0") {
    angularDampingOverride=1
    angularSpringConstant=20000
    bAngularSpring=true
    bOverrideAngularDamping=true
    bUseGravityOverride=true
    gravityScale=0.35
    simulationSpace=RootRelative
    alphaBoolBlend {blendOption=Linear}
    boundBone {boneName=dyn_arm_tassle_l_F_02}
    componentAppliedLinearAccClamp=<3000 3000 3000>
    componentLinearAccScale=<10 10 10>
    componentLinearVelScale=<10 10 10>
    gravityOverride {z=-980}
    physicsBodyDefinitions [{
      boundBone {boneName=dyn_arm_tassle_l_F_02}
      boxExtents=<20 30 30>
      constraintSetup {angularLimitsMax=<15 2 15>, angularLimitsMin=<-15 -25 -15>, angularTarget=<1 0 0>}
      localJointOffset=<4 0 0>
     }]
   }
   n33 = Anim Dynamics - Bone: dyn_arm_tassle_l_G_01(ComponentPose=@n32, Alpha=1, ExternalForce="0,0,0") {
    angularDampingOverride=1
    angularSpringConstant=20000
    bAngularSpring=true
    bOverrideAngularDamping=true
    bUseGravityOverride=true
    gravityScale=0.35
    simulationSpace=RootRelative
    alphaBoolBlend {blendOption=Linear}
    boundBone {boneName=dyn_arm_tassle_l_G_01}
    componentAppliedLinearAccClamp=<3000 3000 3000>
    componentLinearAccScale=<10 10 10>
    componentLinearVelScale=<10 10 10>
    gravityOverride {z=-980}
    physicsBodyDefinitions [{
      boundBone {boneName=dyn_arm_tassle_l_G_01}
      boxExtents=<18 26 27>
      constraintSetup {angularLimitsMax=<15 2 15>, angularLimitsMin=<-15 -25 -15>, angularTarget=<1 0 0>}
      localJointOffset=<4 0 0>
     }]
   }
   n34 = Anim Dynamics - Bone: dyn_arm_tassle_l_G_02(ComponentPose=@n33, Alpha=1, ExternalForce="0,0,0") {
    angularDampingOverride=1
    angularSpringConstant=20000
    bAngularSpring=true
    bOverrideAngularDamping=true
    bUseGravityOverride=true
    gravityScale=0.35
    simulationSpace=RootRelative
    alphaBoolBlend {blendOption=Linear}
    boundBone {boneName=dyn_arm_tassle_l_G_02}
    componentAppliedLinearAccClamp=<3000 3000 3000>
    componentLinearAccScale=<10 10 10>
    componentLinearVelScale=<10 10 10>
    gravityOverride {z=-980}
    physicsBodyDefinitions [{
      boundBone {boneName=dyn_arm_tassle_l_G_02}
      boxExtents=<20 30 30>
      constraintSetup {angularLimitsMax=<15 2 15>, angularLimitsMin=<-15 -25 -15>, angularTarget=<1 0 0>}
      localJointOffset=<4 0 0>
     }]
   }
   n35 = Anim Dynamics - Bone: dyn_arm_tassle_l_H_01(ComponentPose=@n34, Alpha=1, ExternalForce="0,0,0") {
    angularDampingOverride=1
    angularSpringConstant=20000
    bAngularSpring=true
    bOverrideAngularDamping=true
    bUseGravityOverride=true
    gravityScale=0.35
    simulationSpace=RootRelative
    alphaBoolBlend {blendOption=Linear}
    boundBone {boneName=dyn_arm_tassle_l_H_01}
    componentAppliedLinearAccClamp=<3000 3000 3000>
    componentLinearAccScale=<10 10 10>
    componentLinearVelScale=<10 10 10>
    gravityOverride {z=-980}
    physicsBodyDefinitions [{
      boundBone {boneName=dyn_arm_tassle_l_H_01}
      boxExtents=<20 30 30>
      constraintSetup {angularLimitsMax=<15 2 15>, angularLimitsMin=<-15 -25 -15>, angularTarget=<1 0 0>}
      localJointOffset=<4 0 0>
     }]
   }
   n36 = Anim Dynamics - Bone: dyn_arm_tassle_l_H_02(ComponentPose=@n35, Alpha=1, ExternalForce="0,0,0") {
    angularDampingOverride=1
    angularSpringConstant=20000
    bAngularSpring=true
    bOverrideAngularDamping=true
    bUseGravityOverride=true
    gravityScale=0.35
    simulationSpace=RootRelative
    alphaBoolBlend {blendOption=Linear}
    boundBone {boneName=dyn_arm_tassle_l_H_02}
    componentAppliedLinearAccClamp=<3000 3000 3000>
    componentLinearAccScale=<10 10 10>
    componentLinearVelScale=<10 10 10>
    gravityOverride {z=-980}
    physicsBodyDefinitions [{
      boundBone {boneName=dyn_arm_tassle_l_H_02}
      boxExtents=<22 29 31>
      constraintSetup {angularLimitsMax=<15 2 15>, angularLimitsMin=<-15 -25 -15>, angularTarget=<1 0 0>}
      localJointOffset=<4 0 0>
     }]
   }
   n37 = Anim Dynamics - Bone: dyn_arm_tassle_l_J_01(ComponentPose=@n36, Alpha=1, ExternalForce="0,0,0") {
    angularDampingOverride=1
    angularSpringConstant=20000
    bAngularSpring=true
    bOverrideAngularDamping=true
    bUseGravityOverride=true
    gravityScale=0.35
    simulationSpace=RootRelative
    alphaBoolBlend {blendOption=Linear}
    boundBone {boneName=dyn_arm_tassle_l_J_01}
    componentAppliedLinearAccClamp=<3000 3000 3000>
    componentLinearAccScale=<10 10 10>
    componentLinearVelScale=<10 10 10>
    gravityOverride {z=-980}
    physicsBodyDefinitions [{
      boundBone {boneName=dyn_arm_tassle_l_J_01}
      boxExtents=<20 30 30>
      constraintSetup {angularLimitsMax=<15 2 15>, angularLimitsMin=<-15 -25 -15>, angularTarget=<1 0 0>}
      localJointOffset=<4 0 0>
     }]
   }
   n38 = Anim Dynamics - Bone: dyn_arm_tassle_l_J_02(ComponentPose=@n37, Alpha=1, ExternalForce="0,0,0") {
    angularDampingOverride=1
    angularSpringConstant=20000
    bAngularSpring=true
    bOverrideAngularDamping=true
    bUseGravityOverride=true
    gravityScale=0.35
    simulationSpace=RootRelative
    alphaBoolBlend {blendOption=Linear}
    boundBone {boneName=dyn_arm_tassle_l_J_02}
    componentAppliedLinearAccClamp=<3000 3000 3000>
    componentLinearAccScale=<10 10 10>
    componentLinearVelScale=<10 10 10>
    gravityOverride {z=-980}
    physicsBodyDefinitions [{
      boundBone {boneName=dyn_arm_tassle_l_J_02}
      boxExtents=<22 29 31>
      constraintSetup {angularLimitsMax=<15 2 15>, angularLimitsMin=<-15 -25 -15>, angularTarget=<1 0 0>}
      localJointOffset=<4 0 0>
     }]
   }
   n39 = Anim Dynamics - Bone: dyn_arm_tassle_l_K_01(ComponentPose=@n38, Alpha=1, ExternalForce="0,0,0") {
    angularDampingOverride=1
    angularSpringConstant=20000
    bAngularSpring=true
    bOverrideAngularDamping=true
    bUseGravityOverride=true
    gravityScale=0.35
    simulationSpace=RootRelative
    alphaBoolBlend {blendOption=Linear}
    boundBone {boneName=dyn_arm_tassle_l_K_01}
    componentAppliedLinearAccClamp=<3000 3000 3000>
    componentLinearAccScale=<10 10 10>
    componentLinearVelScale=<10 10 10>
    gravityOverride {z=-980}
    physicsBodyDefinitions [{
      boundBone {boneName=dyn_arm_tassle_l_K_01}
      boxExtents=<20 30 30>
      constraintSetup {angularLimitsMax=<15 2 15>, angularLimitsMin=<-15 -25 -15>, angularTarget=<1 0 0>}
      localJointOffset=<4 0 0>
     }]
   }
   n40 = Anim Dynamics - Bone: dyn_arm_tassle_l_K_02(ComponentPose=@n39, Alpha=1, ExternalForce="0,0,0") {
    angularDampingOverride=1
    angularSpringConstant=20000
    bAngularSpring=true
    bOverrideAngularDamping=true
    bUseGravityOverride=true
    gravityScale=0.35
    simulationSpace=RootRelative
    alphaBoolBlend {blendOption=Linear}
    boundBone {boneName=dyn_arm_tassle_l_K_02}
    componentAppliedLinearAccClamp=<3000 3000 3000>
    componentLinearAccScale=<10 10 10>
    componentLinearVelScale=<10 10 10>
    gravityOverride {z=-980}
    physicsBodyDefinitions [{
      boundBone {boneName=dyn_arm_tassle_l_K_02}
      boxExtents=<20 30 30>
      constraintSetup {angularLimitsMax=<15 2 15>, angularLimitsMin=<-15 -25 -15>, angularTarget=<1 0 0>}
      localJointOffset=<4 0 0>
     }]
   }
   n41 = Anim Dynamics - Bone: dyn_arm_tassle_l_L_01(ComponentPose=@n40, Alpha=1, ExternalForce="0,0,0") {
    angularDampingOverride=1
    angularSpringConstant=20000
    bAngularSpring=true
    bOverrideAngularDamping=true
    bUseGravityOverride=true
    gravityScale=0.35
    simulationSpace=RootRelative
    alphaBoolBlend {blendOption=Linear}
    boundBone {boneName=dyn_arm_tassle_l_L_01}
    componentAppliedLinearAccClamp=<3000 3000 3000>
    componentLinearAccScale=<10 10 10>
    componentLinearVelScale=<10 10 10>
    gravityOverride {z=-980}
    physicsBodyDefinitions [{
      boundBone {boneName=dyn_arm_tassle_l_L_01}
      boxExtents=<20 30 30>
      constraintSetup {angularLimitsMax=<15 2 15>, angularLimitsMin=<-15 -25 -15>, angularTarget=<1 0 0>}
      localJointOffset=<4 0 0>
     }]
   }
   n42 = Anim Dynamics - Bone: dyn_arm_tassle_l_L_02(ComponentPose=@n41, Alpha=1, ExternalForce="0,0,0") {
    angularDampingOverride=1
    angularSpringConstant=20000
    bAngularSpring=true
    bOverrideAngularDamping=true
    bUseGravityOverride=true
    gravityScale=0.35
    simulationSpace=RootRelative
    alphaBoolBlend {blendOption=Linear}
    boundBone {boneName=dyn_arm_tassle_l_L_02}
    componentAppliedLinearAccClamp=<3000 3000 3000>
    componentLinearAccScale=<10 10 10>
    componentLinearVelScale=<10 10 10>
    gravityOverride {z=-980}
    physicsBodyDefinitions [{
      boundBone {boneName=dyn_arm_tassle_l_L_02}
      boxExtents=<20 30 30>
      constraintSetup {angularLimitsMax=<15 2 15>, angularLimitsMin=<-15 -25 -15>, angularTarget=<1 0 0>}
      localJointOffset=<4 0 0>
     }]
   }
   n43 = Anim Dynamics - Bone: dyn_arm_tassle_l_M_01(ComponentPose=@n42, Alpha=1, ExternalForce="0,0,0") {
    angularDampingOverride=1
    angularSpringConstant=20000
    bAngularSpring=true
    bOverrideAngularDamping=true
    bUseGravityOverride=true
    gravityScale=0.35
    simulationSpace=RootRelative
    alphaBoolBlend {blendOption=Linear}
    boundBone {boneName=dyn_arm_tassle_l_M_01}
    componentAppliedLinearAccClamp=<3000 3000 3000>
    componentLinearAccScale=<10 10 10>
    componentLinearVelScale=<10 10 10>
    gravityOverride {z=-980}
    physicsBodyDefinitions [{
      boundBone {boneName=dyn_arm_tassle_l_M_01}
      boxExtents=<20 30 30>
      constraintSetup {angularLimitsMax=<15 2 15>, angularLimitsMin=<-15 -25 -15>, angularTarget=<1 0 0>}
      localJointOffset=<4 0 0>
     }]
   }
   n44 = Anim Dynamics - Bone: dyn_arm_tassle_l_M_02(ComponentPose=@n43, Alpha=1, ExternalForce="0,0,0") {
    angularDampingOverride=1
    angularSpringConstant=20000
    bAngularSpring=true
    bOverrideAngularDamping=true
    bUseGravityOverride=true
    gravityScale=0.35
    simulationSpace=RootRelative
    alphaBoolBlend {blendOption=Linear}
    boundBone {boneName=dyn_arm_tassle_l_M_02}
    componentAppliedLinearAccClamp=<3000 3000 3000>
    componentLinearAccScale=<10 10 10>
    componentLinearVelScale=<10 10 10>
    gravityOverride {z=-980}
    physicsBodyDefinitions [{
      boundBone {boneName=dyn_arm_tassle_l_M_02}
      boxExtents=<18 26 27>
      constraintSetup {angularLimitsMax=<15 2 15>, angularLimitsMin=<-15 -25 -15>, angularTarget=<1 0 0>}
      localJointOffset=<4 0 0>
     }]
   }
   n45 = Anim Dynamics - Bone: dyn_arm_tassle_l_N_01(ComponentPose=@n44, Alpha=1, ExternalForce="0,0,0") {
    angularDampingOverride=1
    angularSpringConstant=20000
    bAngularSpring=true
    bOverrideAngularDamping=true
    bUseGravityOverride=true
    gravityScale=0.35
    simulationSpace=RootRelative
    alphaBoolBlend {blendOption=Linear}
    boundBone {boneName=dyn_arm_tassle_l_N_01}
    componentAppliedLinearAccClamp=<3000 3000 3000>
    componentLinearAccScale=<10 10 10>
    componentLinearVelScale=<10 10 10>
    gravityOverride {z=-980}
    physicsBodyDefinitions [{
      boundBone {boneName=dyn_arm_tassle_l_N_01}
      boxExtents=<22 29 31>
      constraintSetup {angularLimitsMax=<15 2 15>, angularLimitsMin=<-15 -25 -15>, angularTarget=<1 0 0>}
      localJointOffset=<4 0 0>
     }]
   }
   n46 = Anim Dynamics - Bone: dyn_arm_tassle_l_N_02(ComponentPose=@n45, Alpha=1, ExternalForce="0,0,0") {
    angularDampingOverride=1
    angularSpringConstant=20000
    bAngularSpring=true
    bOverrideAngularDamping=true
    bUseGravityOverride=true
    gravityScale=0.35
    simulationSpace=RootRelative
    alphaBoolBlend {blendOption=Linear}
    boundBone {boneName=dyn_arm_tassle_l_N_02}
    componentAppliedLinearAccClamp=<3000 3000 3000>
    componentLinearAccScale=<10 10 10>
    componentLinearVelScale=<10 10 10>
    gravityOverride {z=-980}
    physicsBodyDefinitions [{
      boundBone {boneName=dyn_arm_tassle_l_N_02}
      boxExtents=<18 26 27>
      constraintSetup {angularLimitsMax=<15 2 15>, angularLimitsMin=<-15 -25 -15>, angularTarget=<1 0 0>}
      localJointOffset=<4 0 0>
     }]
   }
   n47 = Anim Dynamics - Bone: dyn_arm_tassle_l_O_01(ComponentPose=@n46, Alpha=1, ExternalForce="0,0,0") {
    angularDampingOverride=1
    angularSpringConstant=20000
    bAngularSpring=true
    bOverrideAngularDamping=true
    bUseGravityOverride=true
    gravityScale=0.35
    simulationSpace=RootRelative
    alphaBoolBlend {blendOption=Linear}
    boundBone {boneName=dyn_arm_tassle_l_O_01}
    componentAppliedLinearAccClamp=<3000 3000 3000>
    componentLinearAccScale=<10 10 10>
    componentLinearVelScale=<10 10 10>
    gravityOverride {z=-980}
    physicsBodyDefinitions [{
      boundBone {boneName=dyn_arm_tassle_l_O_01}
      boxExtents=<20 30 30>
      constraintSetup {angularLimitsMax=<15 2 15>, angularLimitsMin=<-15 -25 -15>, angularTarget=<1 0 0>}
      localJointOffset=<4 0 0>
     }]
   }
   n48 = Anim Dynamics - Bone: dyn_arm_tassle_l_O_02(ComponentPose=@n47, Alpha=1, ExternalForce="0,0,0") {
    angularDampingOverride=1
    angularSpringConstant=20000
    bAngularSpring=true
    bOverrideAngularDamping=true
    bUseGravityOverride=true
    gravityScale=0.35
    simulationSpace=RootRelative
    alphaBoolBlend {blendOption=Linear}
    boundBone {boneName=dyn_arm_tassle_l_O_02}
    componentAppliedLinearAccClamp=<3000 3000 3000>
    componentLinearAccScale=<10 10 10>
    componentLinearVelScale=<10 10 10>
    gravityOverride {z=-980}
    physicsBodyDefinitions [{
      boundBone {boneName=dyn_arm_tassle_l_O_02}
      boxExtents=<22 29 31>
      constraintSetup {angularLimitsMax=<15 2 15>, angularLimitsMin=<-15 -25 -15>, angularTarget=<1 0 0>}
      localJointOffset=<4 0 0>
     }]
   }
   n49 = Anim Dynamics - Bone: dyn_arm_tassle_l_P_01(ComponentPose=@n48, Alpha=1, ExternalForce="0,0,0") {
    angularDampingOverride=1
    angularSpringConstant=20000
    bAngularSpring=true
    bOverrideAngularDamping=true
    bUseGravityOverride=true
    gravityScale=0.35
    simulationSpace=RootRelative
    alphaBoolBlend {blendOption=Linear}
    boundBone {boneName=dyn_arm_tassle_l_P_01}
    componentAppliedLinearAccClamp=<3000 3000 3000>
    componentLinearAccScale=<10 10 10>
    componentLinearVelScale=<10 10 10>
    gravityOverride {z=-980}
    physicsBodyDefinitions [{
      boundBone {boneName=dyn_arm_tassle_l_P_01}
      boxExtents=<20 30 30>
      constraintSetup {angularLimitsMax=<15 2 15>, angularLimitsMin=<-15 -25 -15>, angularTarget=<1 0 0>}
      localJointOffset=<4 0 0>
     }]
   }
   n50 = Anim Dynamics - Bone: dyn_arm_tassle_l_P_02(ComponentPose=@n49, Alpha=1, ExternalForce="0,0,0") {
    angularDampingOverride=1
    angularSpringConstant=20000
    bAngularSpring=true
    bOverrideAngularDamping=true
    bUseGravityOverride=true
    gravityScale=0.35
    simulationSpace=RootRelative
    alphaBoolBlend {blendOption=Linear}
    boundBone {boneName=dyn_arm_tassle_l_P_02}
    componentAppliedLinearAccClamp=<3000 3000 3000>
    componentLinearAccScale=<10 10 10>
    componentLinearVelScale=<10 10 10>
    gravityOverride {z=-980}
    physicsBodyDefinitions [{
      boundBone {boneName=dyn_arm_tassle_l_P_02}
      boxExtents=<20 30 30>
      constraintSetup {angularLimitsMax=<15 2 15>, angularLimitsMin=<-15 -25 -15>, angularTarget=<1 0 0>}
      localJointOffset=<4 0 0>
     }]
   }
   n51 = Anim Dynamics - Bone: dyn_arm_tassle_r_A_01(ComponentPose=@n50, Alpha=1, ExternalForce="0,0,0") {
    angularDampingOverride=1
    angularSpringConstant=20000
    bAngularSpring=true
    bOverrideAngularDamping=true
    bUseGravityOverride=true
    gravityScale=0.35
    simulationSpace=RootRelative
    alphaBoolBlend {blendOption=Linear}
    boundBone {boneName=dyn_arm_tassle_r_A_01}
    componentAppliedLinearAccClamp=<3000 3000 3000>
    componentLinearAccScale=<10 10 10>
    componentLinearVelScale=<10 10 10>
    gravityOverride {z=-980}
    physicsBodyDefinitions [{
      boundBone {boneName=dyn_arm_tassle_r_A_01}
      boxExtents=<22 29 31>
      constraintSetup {angularLimitsMax=<15 2 15>, angularLimitsMin=<-15 -25 -15>, angularTarget=<1 0 0>}
      localJointOffset=<4 0 0>
     }]
   }
   n52 = Anim Dynamics - Bone: dyn_arm_tassle_r_A_02(ComponentPose=@n51, Alpha=1, ExternalForce="0,0,0") {
    angularDampingOverride=1
    angularSpringConstant=20000
    bAngularSpring=true
    bOverrideAngularDamping=true
    bUseGravityOverride=true
    gravityScale=0.35
    simulationSpace=RootRelative
    alphaBoolBlend {blendOption=Linear}
    boundBone {boneName=dyn_arm_tassle_r_A_02}
    componentAppliedLinearAccClamp=<3000 3000 3000>
    componentLinearAccScale=<10 10 10>
    componentLinearVelScale=<10 10 10>
    gravityOverride {z=-980}
    physicsBodyDefinitions [{
      boundBone {boneName=dyn_arm_tassle_r_A_02}
      boxExtents=<18 26 27>
      constraintSetup {angularLimitsMax=<15 2 15>, angularLimitsMin=<-15 -25 -15>, angularTarget=<1 0 0>}
      localJointOffset=<4 0 0>
     }]
   }
   n53 = Anim Dynamics - Bone: dyn_arm_tassle_r_B_01(ComponentPose=@n52, Alpha=1, ExternalForce="0,0,0") {
    angularDampingOverride=1
    angularSpringConstant=20000
    bAngularSpring=true
    bOverrideAngularDamping=true
    bUseGravityOverride=true
    gravityScale=0.35
    simulationSpace=RootRelative
    alphaBoolBlend {blendOption=Linear}
    boundBone {boneName=dyn_arm_tassle_r_B_01}
    componentAppliedLinearAccClamp=<3000 3000 3000>
    componentLinearAccScale=<10 10 10>
    componentLinearVelScale=<10 10 10>
    gravityOverride {z=-980}
    physicsBodyDefinitions [{
      boundBone {boneName=dyn_arm_tassle_r_B_01}
      boxExtents=<20 30 30>
      constraintSetup {angularLimitsMax=<15 2 15>, angularLimitsMin=<-15 -25 -15>, angularTarget=<1 0 0>}
      localJointOffset=<4 0 0>
     }]
   }
   n54 = Anim Dynamics - Bone: dyn_arm_tassle_r_B_02(ComponentPose=@n53, Alpha=1, ExternalForce="0,0,0") {
    angularDampingOverride=1
    angularSpringConstant=20000
    bAngularSpring=true
    bOverrideAngularDamping=true
    bUseGravityOverride=true
    gravityScale=0.35
    simulationSpace=RootRelative
    alphaBoolBlend {blendOption=Linear}
    boundBone {boneName=dyn_arm_tassle_r_B_02}
    componentAppliedLinearAccClamp=<3000 3000 3000>
    componentLinearAccScale=<10 10 10>
    componentLinearVelScale=<10 10 10>
    gravityOverride {z=-980}
    physicsBodyDefinitions [{
      boundBone {boneName=dyn_arm_tassle_r_B_02}
      boxExtents=<20 30 30>
      constraintSetup {angularLimitsMax=<15 2 15>, angularLimitsMin=<-15 -25 -15>, angularTarget=<1 0 0>}
      localJointOffset=<4 0 0>
     }]
   }
   n55 = Anim Dynamics - Bone: dyn_arm_tassle_r_C_01(ComponentPose=@n54, Alpha=1, ExternalForce="0,0,0") {
    angularDampingOverride=1
    angularSpringConstant=20000
    bAngularSpring=true
    bOverrideAngularDamping=true
    bUseGravityOverride=true
    gravityScale=0.35
    simulationSpace=RootRelative
    alphaBoolBlend {blendOption=Linear}
    boundBone {boneName=dyn_arm_tassle_r_C_01}
    componentAppliedLinearAccClamp=<3000 3000 3000>
    componentLinearAccScale=<10 10 10>
    componentLinearVelScale=<10 10 10>
    gravityOverride {z=-980}
    physicsBodyDefinitions [{
      boundBone {boneName=dyn_arm_tassle_r_C_01}
      boxExtents=<20 30 30>
      constraintSetup {angularLimitsMax=<15 2 15>, angularLimitsMin=<-15 -25 -15>, angularTarget=<1 0 0>}
      localJointOffset=<4 0 0>
     }]
   }
   n56 = Anim Dynamics - Bone: dyn_arm_tassle_r_C_02(ComponentPose=@n55, Alpha=1, ExternalForce="0,0,0") {
    angularDampingOverride=1
    angularSpringConstant=20000
    bAngularSpring=true
    bOverrideAngularDamping=true
    bUseGravityOverride=true
    gravityScale=0.35
    simulationSpace=RootRelative
    alphaBoolBlend {blendOption=Linear}
    boundBone {boneName=dyn_arm_tassle_r_C_02}
    componentAppliedLinearAccClamp=<3000 3000 3000>
    componentLinearAccScale=<10 10 10>
    componentLinearVelScale=<10 10 10>
    gravityOverride {z=-980}
    physicsBodyDefinitions [{
      boundBone {boneName=dyn_arm_tassle_r_C_02}
      boxExtents=<20 30 30>
      constraintSetup {angularLimitsMax=<15 2 15>, angularLimitsMin=<-15 -25 -15>, angularTarget=<1 0 0>}
      localJointOffset=<4 0 0>
     }]
   }
   n57 = Anim Dynamics - Bone: dyn_arm_tassle_r_D_01(ComponentPose=@n56, Alpha=1, ExternalForce="0,0,0") {
    angularDampingOverride=1
    angularSpringConstant=20000
    bAngularSpring=true
    bOverrideAngularDamping=true
    bUseGravityOverride=true
    gravityScale=0.35
    simulationSpace=RootRelative
    alphaBoolBlend {blendOption=Linear}
    boundBone {boneName=dyn_arm_tassle_r_D_01}
    componentAppliedLinearAccClamp=<3000 3000 3000>
    componentLinearAccScale=<10 10 10>
    componentLinearVelScale=<10 10 10>
    gravityOverride {z=-980}
    physicsBodyDefinitions [{
      boundBone {boneName=dyn_arm_tassle_r_D_01}
      boxExtents=<22 29 31>
      constraintSetup {angularLimitsMax=<15 2 15>, angularLimitsMin=<-15 -25 -15>, angularTarget=<1 0 0>}
      localJointOffset=<4 0 0>
     }]
   }
   n58 = Anim Dynamics - Bone: dyn_arm_tassle_r_D_02(ComponentPose=@n57, Alpha=1, ExternalForce="0,0,0") {
    angularDampingOverride=1
    angularSpringConstant=20000
    bAngularSpring=true
    bOverrideAngularDamping=true
    bUseGravityOverride=true
    gravityScale=0.35
    simulationSpace=RootRelative
    alphaBoolBlend {blendOption=Linear}
    boundBone {boneName=dyn_arm_tassle_r_D_02}
    componentAppliedLinearAccClamp=<3000 3000 3000>
    componentLinearAccScale=<10 10 10>
    componentLinearVelScale=<10 10 10>
    gravityOverride {z=-980}
    physicsBodyDefinitions [{
      boundBone {boneName=dyn_arm_tassle_r_D_02}
      boxExtents=<18 26 27>
      constraintSetup {angularLimitsMax=<15 2 15>, angularLimitsMin=<-15 -25 -15>, angularTarget=<1 0 0>}
      localJointOffset=<4 0 0>
     }]
   }
   n59 = Anim Dynamics - Bone: dyn_arm_tassle_r_E_01(ComponentPose=@n58, Alpha=1, ExternalForce="0,0,0") {
    angularDampingOverride=1
    angularSpringConstant=20000
    bAngularSpring=true
    bOverrideAngularDamping=true
    bUseGravityOverride=true
    gravityScale=0.35
    simulationSpace=RootRelative
    alphaBoolBlend {blendOption=Linear}
    boundBone {boneName=dyn_arm_tassle_r_E_01}
    componentAppliedLinearAccClamp=<3000 3000 3000>
    componentLinearAccScale=<10 10 10>
    componentLinearVelScale=<10 10 10>
    gravityOverride {z=-980}
    physicsBodyDefinitions [{
      boundBone {boneName=dyn_arm_tassle_r_E_01}
      boxExtents=<20 30 30>
      constraintSetup {angularLimitsMax=<15 2 15>, angularLimitsMin=<-15 -25 -15>, angularTarget=<1 0 0>}
      localJointOffset=<4 0 0>
     }]
   }
   n60 = Anim Dynamics - Bone: dyn_arm_tassle_r_E_02(ComponentPose=@n59, Alpha=1, ExternalForce="0,0,0") {
    angularDampingOverride=1
    angularSpringConstant=20000
    bAngularSpring=true
    bOverrideAngularDamping=true
    bUseGravityOverride=true
    gravityScale=0.35
    simulationSpace=RootRelative
    alphaBoolBlend {blendOption=Linear}
    boundBone {boneName=dyn_arm_tassle_r_E_02}
    componentAppliedLinearAccClamp=<3000 3000 3000>
    componentLinearAccScale=<10 10 10>
    componentLinearVelScale=<10 10 10>
    gravityOverride {z=-980}
    physicsBodyDefinitions [{
      boundBone {boneName=dyn_arm_tassle_r_E_02}
      boxExtents=<20 30 30>
      constraintSetup {angularLimitsMax=<15 2 15>, angularLimitsMin=<-15 -25 -15>, angularTarget=<1 0 0>}
      localJointOffset=<4 0 0>
     }]
   }
   n61 = Anim Dynamics - Bone: dyn_arm_tassle_r_F_01(ComponentPose=@n60, Alpha=1, ExternalForce="0,0,0") {
    angularDampingOverride=1
    angularSpringConstant=20000
    bAngularSpring=true
    bOverrideAngularDamping=true
    bUseGravityOverride=true
    gravityScale=0.35
    simulationSpace=RootRelative
    alphaBoolBlend {blendOption=Linear}
    boundBone {boneName=dyn_arm_tassle_r_F_01}
    componentAppliedLinearAccClamp=<3000 3000 3000>
    componentLinearAccScale=<10 10 10>
    componentLinearVelScale=<10 10 10>
    gravityOverride {z=-980}
    physicsBodyDefinitions [{
      boundBone {boneName=dyn_arm_tassle_r_F_01}
      boxExtents=<22 29 31>
      constraintSetup {angularLimitsMax=<15 2 15>, angularLimitsMin=<-15 -25 -15>, angularTarget=<1 0 0>}
      localJointOffset=<4 0 0>
     }]
   }
   n62 = Anim Dynamics - Bone: dyn_arm_tassle_r_F_02(ComponentPose=@n61, Alpha=1, ExternalForce="0,0,0") {
    angularDampingOverride=1
    angularSpringConstant=20000
    bAngularSpring=true
    bOverrideAngularDamping=true
    bUseGravityOverride=true
    gravityScale=0.35
    simulationSpace=RootRelative
    alphaBoolBlend {blendOption=Linear}
    boundBone {boneName=dyn_arm_tassle_r_F_02}
    componentAppliedLinearAccClamp=<3000 3000 3000>
    componentLinearAccScale=<10 10 10>
    componentLinearVelScale=<10 10 10>
    gravityOverride {z=-980}
    physicsBodyDefinitions [{
      boundBone {boneName=dyn_arm_tassle_r_F_02}
      boxExtents=<20 30 30>
      constraintSetup {angularLimitsMax=<15 2 15>, angularLimitsMin=<-15 -25 -15>, angularTarget=<1 0 0>}
      localJointOffset=<4 0 0>
     }]
   }
   n63 = Anim Dynamics - Bone: dyn_arm_tassle_r_G_01(ComponentPose=@n62, Alpha=1, ExternalForce="0,0,0") {
    angularDampingOverride=1
    angularSpringConstant=20000
    bAngularSpring=true
    bOverrideAngularDamping=true
    bUseGravityOverride=true
    gravityScale=0.35
    simulationSpace=RootRelative
    alphaBoolBlend {blendOption=Linear}
    boundBone {boneName=dyn_arm_tassle_r_G_01}
    componentAppliedLinearAccClamp=<3000 3000 3000>
    componentLinearAccScale=<10 10 10>
    componentLinearVelScale=<10 10 10>
    gravityOverride {z=-980}
    physicsBodyDefinitions [{
      boundBone {boneName=dyn_arm_tassle_r_G_01}
      boxExtents=<20 30 30>
      constraintSetup {angularLimitsMax=<15 2 15>, angularLimitsMin=<-15 -25 -15>, angularTarget=<1 0 0>}
      localJointOffset=<4 0 0>
     }]
   }
   n64 = Anim Dynamics - Bone: dyn_arm_tassle_r_G_02(ComponentPose=@n63, Alpha=1, ExternalForce="0,0,0") {
    angularDampingOverride=1
    angularSpringConstant=20000
    bAngularSpring=true
    bOverrideAngularDamping=true
    bUseGravityOverride=true
    gravityScale=0.35
    simulationSpace=RootRelative
    alphaBoolBlend {blendOption=Linear}
    boundBone {boneName=dyn_arm_tassle_r_G_02}
    componentAppliedLinearAccClamp=<3000 3000 3000>
    componentLinearAccScale=<10 10 10>
    componentLinearVelScale=<10 10 10>
    gravityOverride {z=-980}
    physicsBodyDefinitions [{
      boundBone {boneName=dyn_arm_tassle_r_G_02}
      boxExtents=<18 26 27>
      constraintSetup {angularLimitsMax=<15 2 15>, angularLimitsMin=<-15 -25 -15>, angularTarget=<1 0 0>}
      localJointOffset=<4 0 0>
     }]
   }
   n65 = Anim Dynamics - Bone: dyn_arm_tassle_r_H_01(ComponentPose=@n64, Alpha=1, ExternalForce="0,0,0") {
    angularDampingOverride=1
    angularSpringConstant=20000
    bAngularSpring=true
    bOverrideAngularDamping=true
    bUseGravityOverride=true
    gravityScale=0.35
    simulationSpace=RootRelative
    alphaBoolBlend {blendOption=Linear}
    boundBone {boneName=dyn_arm_tassle_r_H_01}
    componentAppliedLinearAccClamp=<3000 3000 3000>
    componentLinearAccScale=<10 10 10>
    componentLinearVelScale=<10 10 10>
    gravityOverride {z=-980}
    physicsBodyDefinitions [{
      boundBone {boneName=dyn_arm_tassle_r_H_01}
      boxExtents=<20 30 30>
      constraintSetup {angularLimitsMax=<15 2 15>, angularLimitsMin=<-15 -25 -15>, angularTarget=<1 0 0>}
      localJointOffset=<4 0 0>
     }]
   }
   n66 = Anim Dynamics - Bone: dyn_arm_tassle_r_H_02(ComponentPose=@n65, Alpha=1, ExternalForce="0,0,0") {
    angularDampingOverride=1
    angularSpringConstant=20000
    bAngularSpring=true
    bOverrideAngularDamping=true
    bUseGravityOverride=true
    gravityScale=0.35
    simulationSpace=RootRelative
    alphaBoolBlend {blendOption=Linear}
    boundBone {boneName=dyn_arm_tassle_r_H_02}
    componentAppliedLinearAccClamp=<3000 3000 3000>
    componentLinearAccScale=<10 10 10>
    componentLinearVelScale=<10 10 10>
    gravityOverride {z=-980}
    physicsBodyDefinitions [{
      boundBone {boneName=dyn_arm_tassle_r_H_02}
      boxExtents=<18 26 27>
      constraintSetup {angularLimitsMax=<15 2 15>, angularLimitsMin=<-15 -25 -15>, angularTarget=<1 0 0>}
      localJointOffset=<4 0 0>
     }]
   }
   n67 = Anim Dynamics - Bone: dyn_arm_tassle_r_J_01(ComponentPose=@n66, Alpha=1, ExternalForce="0,0,0") {
    angularDampingOverride=1
    angularSpringConstant=20000
    bAngularSpring=true
    bOverrideAngularDamping=true
    bUseGravityOverride=true
    gravityScale=0.35
    simulationSpace=RootRelative
    alphaBoolBlend {blendOption=Linear}
    boundBone {boneName=dyn_arm_tassle_r_J_01}
    componentAppliedLinearAccClamp=<3000 3000 3000>
    componentLinearAccScale=<10 10 10>
    componentLinearVelScale=<10 10 10>
    gravityOverride {z=-980}
    physicsBodyDefinitions [{
      boundBone {boneName=dyn_arm_tassle_r_J_01}
      boxExtents=<20 30 30>
      constraintSetup {angularLimitsMax=<15 2 15>, angularLimitsMin=<-15 -25 -15>, angularTarget=<1 0 0>}
      localJointOffset=<4 0 0>
     }]
   }
   n68 = Anim Dynamics - Bone: dyn_arm_tassle_r_J_02(ComponentPose=@n67, Alpha=1, ExternalForce="0,0,0") {
    angularDampingOverride=1
    angularSpringConstant=20000
    bAngularSpring=true
    bOverrideAngularDamping=true
    bUseGravityOverride=true
    gravityScale=0.35
    simulationSpace=RootRelative
    alphaBoolBlend {blendOption=Linear}
    boundBone {boneName=dyn_arm_tassle_r_J_02}
    componentAppliedLinearAccClamp=<3000 3000 3000>
    componentLinearAccScale=<10 10 10>
    componentLinearVelScale=<10 10 10>
    gravityOverride {z=-980}
    physicsBodyDefinitions [{
      boundBone {boneName=dyn_arm_tassle_r_J_02}
      boxExtents=<22 29 31>
      constraintSetup {angularLimitsMax=<15 2 15>, angularLimitsMin=<-15 -25 -15>, angularTarget=<1 0 0>}
      localJointOffset=<4 0 0>
     }]
   }
   n69 = Anim Dynamics - Bone: dyn_arm_tassle_r_K_01(ComponentPose=@n68, Alpha=1, ExternalForce="0,0,0") {
    angularDampingOverride=1
    angularSpringConstant=20000
    bAngularSpring=true
    bOverrideAngularDamping=true
    bUseGravityOverride=true
    gravityScale=0.35
    simulationSpace=RootRelative
    alphaBoolBlend {blendOption=Linear}
    boundBone {boneName=dyn_arm_tassle_r_K_01}
    componentAppliedLinearAccClamp=<3000 3000 3000>
    componentLinearAccScale=<10 10 10>
    componentLinearVelScale=<10 10 10>
    gravityOverride {z=-980}
    physicsBodyDefinitions [{
      boundBone {boneName=dyn_arm_tassle_r_K_01}
      boxExtents=<20 30 30>
      constraintSetup {angularLimitsMax=<15 2 15>, angularLimitsMin=<-15 -25 -15>, angularTarget=<1 0 0>}
      localJointOffset=<4 0 0>
     }]
   }
   n70 = Anim Dynamics - Bone: dyn_arm_tassle_r_K_02(ComponentPose=@n69, Alpha=1, ExternalForce="0,0,0") {
    angularDampingOverride=1
    angularSpringConstant=20000
    bAngularSpring=true
    bOverrideAngularDamping=true
    bUseGravityOverride=true
    gravityScale=0.35
    simulationSpace=RootRelative
    alphaBoolBlend {blendOption=Linear}
    boundBone {boneName=dyn_arm_tassle_r_K_02}
    componentAppliedLinearAccClamp=<3000 3000 3000>
    componentLinearAccScale=<10 10 10>
    componentLinearVelScale=<10 10 10>
    gravityOverride {z=-980}
    physicsBodyDefinitions [{
      boundBone {boneName=dyn_arm_tassle_r_K_02}
      boxExtents=<18 26 27>
      constraintSetup {angularLimitsMax=<15 2 15>, angularLimitsMin=<-15 -25 -15>, angularTarget=<1 0 0>}
      localJointOffset=<4 0 0>
     }]
   }
   n71 = Anim Dynamics - Bone: dyn_arm_tassle_r_L_01(ComponentPose=@n70, Alpha=1, ExternalForce="0,0,0") {
    angularDampingOverride=1
    angularSpringConstant=20000
    bAngularSpring=true
    bOverrideAngularDamping=true
    bUseGravityOverride=true
    gravityScale=0.35
    simulationSpace=RootRelative
    alphaBoolBlend {blendOption=Linear}
    boundBone {boneName=dyn_arm_tassle_r_L_01}
    componentAppliedLinearAccClamp=<3000 3000 3000>
    componentLinearAccScale=<10 10 10>
    componentLinearVelScale=<10 10 10>
    gravityOverride {z=-980}
    physicsBodyDefinitions [{
      boundBone {boneName=dyn_arm_tassle_r_L_01}
      boxExtents=<20 30 30>
      constraintSetup {angularLimitsMax=<15 2 15>, angularLimitsMin=<-15 -25 -15>, angularTarget=<1 0 0>}
      localJointOffset=<4 0 0>
     }]
   }
   n72 = Anim Dynamics - Bone: dyn_arm_tassle_r_L_02(ComponentPose=@n71, Alpha=1, ExternalForce="0,0,0") {
    angularDampingOverride=1
    angularSpringConstant=20000
    bAngularSpring=true
    bOverrideAngularDamping=true
    bUseGravityOverride=true
    gravityScale=0.35
    simulationSpace=RootRelative
    alphaBoolBlend {blendOption=Linear}
    boundBone {boneName=dyn_arm_tassle_r_L_02}
    componentAppliedLinearAccClamp=<3000 3000 3000>
    componentLinearAccScale=<10 10 10>
    componentLinearVelScale=<10 10 10>
    gravityOverride {z=-980}
    physicsBodyDefinitions [{
      boundBone {boneName=dyn_arm_tassle_r_L_02}
      boxExtents=<18 26 27>
      constraintSetup {angularLimitsMax=<15 2 15>, angularLimitsMin=<-15 -25 -15>, angularTarget=<1 0 0>}
      localJointOffset=<4 0 0>
     }]
   }
   n73 = Anim Dynamics - Bone: dyn_arm_tassle_r_M_01(ComponentPose=@n72, Alpha=1, ExternalForce="0,0,0") {
    angularDampingOverride=1
    angularSpringConstant=20000
    bAngularSpring=true
    bOverrideAngularDamping=true
    bUseGravityOverride=true
    gravityScale=0.35
    simulationSpace=RootRelative
    alphaBoolBlend {blendOption=Linear}
    boundBone {boneName=dyn_arm_tassle_r_M_01}
    componentAppliedLinearAccClamp=<3000 3000 3000>
    componentLinearAccScale=<10 10 10>
    componentLinearVelScale=<10 10 10>
    gravityOverride {z=-980}
    physicsBodyDefinitions [{
      boundBone {boneName=dyn_arm_tassle_r_M_01}
      boxExtents=<20 30 30>
      constraintSetup {angularLimitsMax=<15 2 15>, angularLimitsMin=<-15 -25 -15>, angularTarget=<1 0 0>}
      localJointOffset=<4 0 0>
     }]
   }
   n74 = Anim Dynamics - Bone: dyn_arm_tassle_r_M_02(ComponentPose=@n73, Alpha=1, ExternalForce="0,0,0") {
    angularDampingOverride=1
    angularSpringConstant=20000
    bAngularSpring=true
    bOverrideAngularDamping=true
    bUseGravityOverride=true
    gravityScale=0.35
    simulationSpace=RootRelative
    alphaBoolBlend {blendOption=Linear}
    boundBone {boneName=dyn_arm_tassle_r_M_02}
    componentAppliedLinearAccClamp=<3000 3000 3000>
    componentLinearAccScale=<10 10 10>
    componentLinearVelScale=<10 10 10>
    gravityOverride {z=-980}
    physicsBodyDefinitions [{
      boundBone {boneName=dyn_arm_tassle_r_M_02}
      boxExtents=<22 29 31>
      constraintSetup {angularLimitsMax=<15 2 15>, angularLimitsMin=<-15 -25 -15>, angularTarget=<1 0 0>}
      localJointOffset=<4 0 0>
     }]
   }
   n75 = Anim Dynamics - Bone: dyn_arm_tassle_r_N_01(ComponentPose=@n74, Alpha=1, ExternalForce="0,0,0") {
    angularDampingOverride=1
    angularSpringConstant=20000
    bAngularSpring=true
    bOverrideAngularDamping=true
    bUseGravityOverride=true
    gravityScale=0.35
    simulationSpace=RootRelative
    alphaBoolBlend {blendOption=Linear}
    boundBone {boneName=dyn_arm_tassle_r_N_01}
    componentAppliedLinearAccClamp=<3000 3000 3000>
    componentLinearAccScale=<10 10 10>
    componentLinearVelScale=<10 10 10>
    gravityOverride {z=-980}
    physicsBodyDefinitions [{
      boundBone {boneName=dyn_arm_tassle_r_N_01}
      boxExtents=<20 30 30>
      constraintSetup {angularLimitsMax=<15 2 15>, angularLimitsMin=<-15 -25 -15>, angularTarget=<1 0 0>}
      localJointOffset=<4 0 0>
     }]
   }
   n76 = Anim Dynamics - Bone: dyn_arm_tassle_r_N_02(ComponentPose=@n75, Alpha=1, ExternalForce="0,0,0") {
    angularDampingOverride=1
    angularSpringConstant=20000
    bAngularSpring=true
    bOverrideAngularDamping=true
    bUseGravityOverride=true
    gravityScale=0.35
    simulationSpace=RootRelative
    alphaBoolBlend {blendOption=Linear}
    boundBone {boneName=dyn_arm_tassle_r_N_02}
    componentAppliedLinearAccClamp=<3000 3000 3000>
    componentLinearAccScale=<10 10 10>
    componentLinearVelScale=<10 10 10>
    gravityOverride {z=-980}
    physicsBodyDefinitions [{
      boundBone {boneName=dyn_arm_tassle_r_N_02}
      boxExtents=<18 26 27>
      constraintSetup {angularLimitsMax=<15 2 15>, angularLimitsMin=<-15 -25 -15>, angularTarget=<1 0 0>}
      localJointOffset=<4 0 0>
     }]
   }
   n77 = Anim Dynamics - Bone: dyn_arm_tassle_r_O_01(ComponentPose=@n76, Alpha=1, ExternalForce="0,0,0") {
    angularDampingOverride=1
    angularSpringConstant=20000
    bAngularSpring=true
    bOverrideAngularDamping=true
    bUseGravityOverride=true
    gravityScale=0.35
    simulationSpace=RootRelative
    alphaBoolBlend {blendOption=Linear}
    boundBone {boneName=dyn_arm_tassle_r_O_01}
    componentAppliedLinearAccClamp=<3000 3000 3000>
    componentLinearAccScale=<10 10 10>
    componentLinearVelScale=<10 10 10>
    gravityOverride {z=-980}
    physicsBodyDefinitions [{
      boundBone {boneName=dyn_arm_tassle_r_O_01}
      boxExtents=<22 29 31>
      constraintSetup {angularLimitsMax=<15 2 15>, angularLimitsMin=<-15 -25 -15>, angularTarget=<1 0 0>}
      localJointOffset=<4 0 0>
     }]
   }
   n78 = Anim Dynamics - Bone: dyn_arm_tassle_r_O_02(ComponentPose=@n77, Alpha=1, ExternalForce="0,0,0") {
    angularDampingOverride=1
    angularSpringConstant=20000
    bAngularSpring=true
    bOverrideAngularDamping=true
    bUseGravityOverride=true
    gravityScale=0.35
    simulationSpace=RootRelative
    alphaBoolBlend {blendOption=Linear}
    boundBone {boneName=dyn_arm_tassle_r_O_02}
    componentAppliedLinearAccClamp=<3000 3000 3000>
    componentLinearAccScale=<10 10 10>
    componentLinearVelScale=<10 10 10>
    gravityOverride {z=-980}
    physicsBodyDefinitions [{
      boundBone {boneName=dyn_arm_tassle_r_O_02}
      boxExtents=<20 30 30>
      constraintSetup {angularLimitsMax=<15 2 15>, angularLimitsMin=<-15 -25 -15>, angularTarget=<1 0 0>}
      localJointOffset=<4 0 0>
     }]
   }
   n79 = Anim Dynamics - Bone: dyn_arm_tassle_r_P_01(ComponentPose=@n78, Alpha=1, ExternalForce="0,0,0") {
    angularDampingOverride=1
    angularSpringConstant=20000
    bAngularSpring=true
    bOverrideAngularDamping=true
    bUseGravityOverride=true
    gravityScale=0.35
    simulationSpace=RootRelative
    alphaBoolBlend {blendOption=Linear}
    boundBone {boneName=dyn_arm_tassle_r_P_01}
    componentAppliedLinearAccClamp=<3000 3000 3000>
    componentLinearAccScale=<10 10 10>
    componentLinearVelScale=<10 10 10>
    gravityOverride {z=-980}
    physicsBodyDefinitions [{
      boundBone {boneName=dyn_arm_tassle_r_P_01}
      boxExtents=<18 26 27>
      constraintSetup {angularLimitsMax=<15 2 15>, angularLimitsMin=<-15 -25 -15>, angularTarget=<1 0 0>}
      localJointOffset=<4 0 0>
     }]
   }
   n80 = Anim Dynamics - Bone: dyn_arm_tassle_r_P_02(ComponentPose=@n79, Alpha=1, ExternalForce="0,0,0") {
    angularDampingOverride=1
    angularSpringConstant=20000
    bAngularSpring=true
    bOverrideAngularDamping=true
    bUseGravityOverride=true
    gravityScale=0.35
    simulationSpace=RootRelative
    alphaBoolBlend {blendOption=Linear}
    boundBone {boneName=dyn_arm_tassle_r_P_02}
    componentAppliedLinearAccClamp=<3000 3000 3000>
    componentLinearAccScale=<10 10 10>
    componentLinearVelScale=<10 10 10>
    gravityOverride {z=-980}
    physicsBodyDefinitions [{
      boundBone {boneName=dyn_arm_tassle_r_P_02}
      boxExtents=<20 30 30>
      constraintSetup {angularLimitsMax=<15 2 15>, angularLimitsMin=<-15 -25 -15>, angularTarget=<1 0 0>}
      localJointOffset=<4 0 0>
     }]
   }
   n81 = Control Rig!(Source=Component To Local(ComponentPose=@n80), Alpha) -> (Pose) {
    alphaBoolBlend {blendOption=Linear}
    controlRigAssetReference {blueprintRigClass=&/F_MED_Scribble_ControlRig_grenades.F_MED_Scribble_ControlRig_grenades_C}
   }
   Output Pose(Result=@n81) {layerGroup=None}
  ]}
 EventGraph {nodes [
   Event BlueprintUpdateAnimation()
   > n1 = Sequence!(execute) -> (then_0, then_1, then_2:Exec, then_3:Exec)
   > n2 = Gravity Override State Function DEPRECATED(
    CharacterPartAnimInstance=Self-Reference()
    GameTimeInSeconds=GetGameTimeInSeconds()
    GravityOverrideParams=Get Gravity Override Params()
    RBAN Params=Get RBAN Params()
    DebugJointName=spine_05
   )
   > Set GravityOverride(GravityOverride=@n2)
   > Set RBAN Settings(RBAN Settings:RBANSetting Struct Structure=@n2.RBAN SettingsOut) -> (Output_Get:RBANSetting Struct Structure)
   n3 = Branch(execute=@n1.then_1, Condition=Get bIsSliding())
   Do Once(execute=@n3.else, Reset=@n3.then, Start Closed=true)
   > Delay(Duration=0.2)
   > ResetDynamics(InTeleportType=TeleportPhysics)
   > ResetDynamics(InTeleportType=ResetPhysics)
   n4 = Vehicle Dynamics Scaler Function!(execute=@n1.then_2, CharacterPartAnimInstance:Custom Character Part Anim Instance Object Reference=Self-Reference(), Debug:bool=false) -> (then, InVehicleHeadScaleAlpha:double, InVehicleHeadSimAlpha:double, InVehicleSkirtScaleAlpha:double, InVehicleSkirtSimAlpha:double)
   > Set in vehicle scale alpha(in vehicle scale alpha:double=@n4.InVehicleSkirtScaleAlpha) -> (Output_Get:double)
   > Set in vehicle rigid alpha(in vehicle rigid alpha:double=@n4.InVehicleSkirtSimAlpha) -> (Output_Get:double)
   DrawVectorFromJointFunction!(
    execute
    SkelMeshComponent:Scene Component Object Reference=GetPartSkeletalMeshComponent()
    JointName:Name=pelvis
    LineEnd:Vector=Break Cloth Setting Struct!(ClothSettingStruct:"Cloth Setting Struct Structure (by ref)") -> (
     LinearVelocityScale_7_#1:Vector=~
     AngularVelocityScale_6_#2:float=~
     FictitiousAngularScale_10_#3:float=~
     GravityOverride_15_#4:Vector=~
     DragCoefficient_23_#5:Vector 2D Structure
     LiftCoefficient_24_#6:Vector 2D Structure
     JointName_27_#7:Name
    ) {is=title K2Node_BreakStruct}.GravityOverride_15_#4
    JointOffset:Vector="0, 0, 0"
    LineColor:Linear Color Structure=(R=1,G=0,B=0,A=1)
    Thickness:double=2
    Foreground:EDrawDebugSceneDepthPriorityGroup Enum=Foreground
   ) -> (then)
  ]}
]
variables [
 Cloth Core Settings {
  type=ClothCoreSettingStruct
  default {
   collisionThickness=0
   dampingCoefficient=0.1
   frictionCoefficient=0.5
   localDampingCoefficient=0.1
   numberIterations=2
   numberSubsteps=1
   selfCollisionThickness=0
   animDriveDamping=<0 1>
   animDriveStiffness=<0 1>
   areaStiffness=<1 1>
   bendingStiffness=<1 1>
   edgeStiffness=<1 1>
   tetherStiffness=<1 1>
  }
 }
 Cloth Core Settings_0 {
  type=ClothCoreSettingStruct
  default {
   collisionThickness=0
   dampingCoefficient=0.1
   frictionCoefficient=0.5
   localDampingCoefficient=0.1
   numberIterations=2
   numberSubsteps=1
   selfCollisionThickness=0
   animDriveDamping=<0 1>
   animDriveStiffness=<0 1>
   areaStiffness=<1 1>
   bendingStiffness=<1 1>
   edgeStiffness=<1 1>
   tetherStiffness=<1 1>
  }
 }
 Cloth Paramaters {
  type=ClothParamsStruct
  default {
   dBNO {
    angularVelocityScale=0.6
    fictitiousAngularScale=0.1
    jointName=pelvis
    dragCoefficient=<0.2 1>
    gravityOverride=<-980 400 0>
    liftCoefficient=<0 0.5>
    linearVelocityScale=<0.3 0.3 0.1>
   }
   dBNO_Carried {
    angularVelocityScale=0.6
    fictitiousAngularScale=0.1
    jointName=pelvis
    dragCoefficient=<0.2 1>
    gravityOverride=<-980 -200 0>
    liftCoefficient=<0 0.5>
    linearVelocityScale=<0.3 0.3 0.1>
   }
   dBNO_Impulse {
    angularVelocityScale=0
    fictitiousAngularScale=0
    jointName=pelvis
    dragCoefficient=<0 0.2>
    gravityOverride=<-3000 -2000 0>
    liftCoefficient=<0 0.1>
    linearVelocityScale=<0.3 0.1 0.3>
   }
   dBNO_RollBwd {
    angularVelocityScale=0
    fictitiousAngularScale=0
    jointName=pelvis
    dragCoefficient=<0 0.2>
    gravityOverride=<-2000 0 0>
    liftCoefficient=<0 0.1>
    linearVelocityScale=<0.3 0.1 0.3>
   }
   dBNO_RollFwd {
    angularVelocityScale=0
    fictitiousAngularScale=0
    jointName=pelvis
    dragCoefficient=<0 0.2>
    gravityOverride=<-980 0 0>
    liftCoefficient=<0 0.1>
    linearVelocityScale=<0.3 0.1 0.3>
   }
   dBNO_Tumbling {
    angularVelocityScale=0
    fictitiousAngularScale=0
    jointName=pelvis
    dragCoefficient=<0 0.2>
    gravityOverride=<-2500 -300 0>
    liftCoefficient=<0 0.1>
    linearVelocityScale=<0.3 0.1 0.3>
   }
   emoteOrMelee {
    angularVelocityScale=0.6
    fictitiousAngularScale=0.1
    jointName=root
    dragCoefficient=<2.5 2.5>
    gravityOverride=<0 0 -980>
    liftCoefficient=<0.5 0.5>
    linearVelocityScale=<0.6 0.6 0.5>
   }
   front_End {
    angularVelocityScale=0.05
    fictitiousAngularScale=0.6
    jointName=root
    dragCoefficient=<0.2 0.2>
    gravityOverride=<0 -200 -980>
    liftCoefficient=<0.1 0.1>
    linearVelocityScale=<0.1 0.1 0.1>
   }
   onGround_Crouch_Moving {
    angularVelocityScale=0.1
    fictitiousAngularScale=0.3
    jointName=root
    dragCoefficient=<2 2>
    gravityOverride=<0 0 -980>
    liftCoefficient=<0.5 0.5>
    linearVelocityScale=<0.6 0.6 0.5>
   }
   onGround_Falling {
    angularVelocityScale=0.6
    fictitiousAngularScale=0
    jointName=root
    dragCoefficient=<0 0.2>
    gravityOverride=<0 -500 -980>
    liftCoefficient=<0 0.1>
    linearVelocityScale=<2 1 0.3>
   }
   onGround_Moving {
    angularVelocityScale=0.6
    fictitiousAngularScale=0
    jointName=root
    dragCoefficient=<0.2 2>
    gravityOverride=<0 0 -980>
    liftCoefficient=<0.1 0.5>
    linearVelocityScale=<0.6 0.6 0.5>
   }
   onGround_Standing {
    angularVelocityScale=0.05
    fictitiousAngularScale=0
    jointName=root
    dragCoefficient=<2.5 2.5>
    gravityOverride=<0 0 -980>
    liftCoefficient=<0.5 0.5>
    linearVelocityScale=<0.6 0.6 0.5>
   }
   rail_Grinding {
    angularVelocityScale=0.2
    fictitiousAngularScale=0
    jointName=root
    dragCoefficient=<0.1 0.2>
    gravityOverride=<50 -500 -400>
    liftCoefficient=<0 0.5>
    linearVelocityScale=<0.1 0.1 0.1>
   }
   riding_Moving {
    angularVelocityScale=0.05
    fictitiousAngularScale=1
    jointName=root
    dragCoefficient=<0.5 0.5>
    gravityOverride=<0 0 -980>
    liftCoefficient=<0.5 0.5>
    linearVelocityScale=<0.4 0.6 0.3>
   }
   riding_Moving_Targeting {
    angularVelocityScale=0.05
    fictitiousAngularScale=0.7
    jointName=root
    dragCoefficient=<0.5 0.5>
    gravityOverride=<0 0 -980>
    liftCoefficient=<0.5 0.5>
    linearVelocityScale=<0.2 0.6 0.2>
   }
   riding_Standing {
    angularVelocityScale=0.05
    fictitiousAngularScale=1
    jointName=root
    dragCoefficient=<0.5 0.5>
    gravityOverride=<0 -450 -980>
    liftCoefficient=<0.5 0.5>
    linearVelocityScale=<0.6 0.6 0.3>
   }
   skydive_DivingDown {
    angularVelocityScale=0.1
    fictitiousAngularScale=0
    jointName=pelvis
    dragCoefficient=<0.3 0.3>
    gravityOverride=<-450 -100 0>
    liftCoefficient=<0.9 0.9>
    linearVelocityScale=<0.1 0.1 0.1>
   }
   skydive_FloatingDown {
    angularVelocityScale=0.1
    fictitiousAngularScale=0
    jointName=root
    dragCoefficient=<0.3 0.3>
    gravityOverride=<0 -800 200>
    liftCoefficient=<1.3 1.3>
    linearVelocityScale=<0.2 0.2 0.1>
   }
   skydive_GoingUp {
    angularVelocityScale=0.1
    fictitiousAngularScale=0
    jointName=root
    dragCoefficient=<0.01 0.01>
    gravityOverride=<0 -2000 400>
    liftCoefficient=<0.05 0.05>
    linearVelocityScale=<0.1 0.1 0.01>
   }
   skydive_Parachuting {
    angularVelocityScale=0.6
    fictitiousAngularScale=0.1
    jointName=root
    dragCoefficient=<0.8 0.8>
    gravityOverride=<0 -600 -500>
    liftCoefficient=<0.8 0.8>
    linearVelocityScale=<0.2 0.2 0.3>
   }
   skydive_Surfing {
    angularVelocityScale=0.6
    fictitiousAngularScale=0.1
    jointName=root
    dragCoefficient=<1.5 1.5>
    gravityOverride=<0 -600 -500>
    liftCoefficient=<1 1>
    linearVelocityScale=<0.1 0.1 0.3>
   }
   swimming {
    angularVelocityScale=0.6
    fictitiousAngularScale=0
    jointName=root
    dragCoefficient=<2.5 2.5>
    gravityOverride=<0 0 0>
    liftCoefficient=<0.1 0.1>
    linearVelocityScale=<0.6 0.6 0.5>
   }
   tactical_Sprint {
    angularVelocityScale=0.6
    fictitiousAngularScale=0.1
    jointName=root
    dragCoefficient=<2 2>
    gravityOverride=<0 0 -980>
    liftCoefficient=<0.5 0.5>
    linearVelocityScale=<0.6 0.6 0.5>
   }
   ziplining {
    angularVelocityScale=0.6
    fictitiousAngularScale=0
    jointName=root
    dragCoefficient=<0 1.5>
    gravityOverride=<0 0 -980>
    liftCoefficient=<0 0.3>
    linearVelocityScale=<0.4 0.4 0.5>
   }
  }
 }
 Cloth Paramaters_0 {
  type=ClothParamsStruct
  default {
   dBNO {
    angularVelocityScale=0.6
    fictitiousAngularScale=0.1
    jointName=pelvis
    dragCoefficient=<0.2 1>
    gravityOverride=<-980 400 0>
    liftCoefficient=<0 0.5>
    linearVelocityScale=<0.3 0.3 0.1>
   }
   dBNO_Carried {
    angularVelocityScale=0.6
    fictitiousAngularScale=0.1
    jointName=pelvis
    dragCoefficient=<0.2 1>
    gravityOverride=<-980 -200 0>
    liftCoefficient=<0 0.5>
    linearVelocityScale=<0.3 0.3 0.1>
   }
   dBNO_Impulse {
    angularVelocityScale=0
    fictitiousAngularScale=0
    jointName=pelvis
    dragCoefficient=<0 0.2>
    gravityOverride=<-3000 -2000 0>
    liftCoefficient=<0 0.1>
    linearVelocityScale=<0.3 0.1 0.3>
   }
   dBNO_RollBwd {
    angularVelocityScale=0
    fictitiousAngularScale=0
    jointName=pelvis
    dragCoefficient=<0 0.2>
    gravityOverride=<-2000 0 0>
    liftCoefficient=<0 0.1>
    linearVelocityScale=<0.3 0.1 0.3>
   }
   dBNO_RollFwd {
    angularVelocityScale=0
    fictitiousAngularScale=0
    jointName=pelvis
    dragCoefficient=<0 0.2>
    gravityOverride=<-980 0 0>
    liftCoefficient=<0 0.1>
    linearVelocityScale=<0.3 0.1 0.3>
   }
   dBNO_Tumbling {
    angularVelocityScale=0
    fictitiousAngularScale=0
    jointName=pelvis
    dragCoefficient=<0 0.2>
    gravityOverride=<-2500 -300 0>
    liftCoefficient=<0 0.1>
    linearVelocityScale=<0.3 0.1 0.3>
   }
   emoteOrMelee {
    angularVelocityScale=0.6
    fictitiousAngularScale=0.1
    jointName=root
    dragCoefficient=<2.5 2.5>
    gravityOverride=<0 0 -980>
    liftCoefficient=<0.5 0.5>
    linearVelocityScale=<0.6 0.6 0.5>
   }
   front_End {
    angularVelocityScale=0.05
    fictitiousAngularScale=0.6
    jointName=root
    dragCoefficient=<0.2 0.2>
    gravityOverride=<0 -200 -980>
    liftCoefficient=<0.1 0.1>
    linearVelocityScale=<0.1 0.1 0.1>
   }
   onGround_Crouch_Moving {
    angularVelocityScale=0.1
    fictitiousAngularScale=0.3
    jointName=root
    dragCoefficient=<2 2>
    gravityOverride=<0 0 -980>
    liftCoefficient=<0.5 0.5>
    linearVelocityScale=<0.6 0.6 0.5>
   }
   onGround_Falling {
    angularVelocityScale=0.6
    fictitiousAngularScale=0
    jointName=root
    dragCoefficient=<0 0.2>
    gravityOverride=<0 -500 -980>
    liftCoefficient=<0 0.1>
    linearVelocityScale=<2 1 0.3>
   }
   onGround_Moving {
    angularVelocityScale=0.6
    fictitiousAngularScale=0
    jointName=root
    dragCoefficient=<2 2>
    gravityOverride=<0 0 -980>
    liftCoefficient=<0.5 0.5>
    linearVelocityScale=<0.6 0.6 0.5>
   }
   onGround_Standing {
    angularVelocityScale=0.05
    fictitiousAngularScale=0
    jointName=root
    dragCoefficient=<2.5 2.5>
    gravityOverride=<0 0 -980>
    liftCoefficient=<0.5 0.5>
    linearVelocityScale=<0.6 0.6 0.5>
   }
   rail_Grinding {
    angularVelocityScale=0.2
    fictitiousAngularScale=0
    jointName=root
    dragCoefficient=<0.1 0.2>
    gravityOverride=<50 -500 -400>
    liftCoefficient=<0 0.5>
    linearVelocityScale=<0.1 0.1 0.1>
   }
   riding_Moving {
    angularVelocityScale=0.05
    fictitiousAngularScale=1
    jointName=root
    dragCoefficient=<0.5 0.5>
    gravityOverride=<0 0 -980>
    liftCoefficient=<0.5 0.5>
    linearVelocityScale=<0.4 0.6 0.3>
   }
   riding_Moving_Targeting {
    angularVelocityScale=0.05
    fictitiousAngularScale=0.7
    jointName=root
    dragCoefficient=<0.5 0.5>
    gravityOverride=<0 0 -980>
    liftCoefficient=<0.5 0.5>
    linearVelocityScale=<0.2 0.6 0.2>
   }
   riding_Standing {
    angularVelocityScale=0.05
    fictitiousAngularScale=1
    jointName=root
    dragCoefficient=<0.5 0.5>
    gravityOverride=<0 -450 -980>
    liftCoefficient=<0.5 0.5>
    linearVelocityScale=<0.6 0.6 0.3>
   }
   skydive_DivingDown {
    angularVelocityScale=0.1
    fictitiousAngularScale=0
    jointName=pelvis
    dragCoefficient=<0.3 0.3>
    gravityOverride=<-450 -100 0>
    liftCoefficient=<0.9 0.9>
    linearVelocityScale=<0.1 0.1 0.1>
   }
   skydive_FloatingDown {
    angularVelocityScale=0.1
    fictitiousAngularScale=0
    jointName=root
    dragCoefficient=<0.3 0.3>
    gravityOverride=<0 -800 200>
    liftCoefficient=<1.3 1.3>
    linearVelocityScale=<0.2 0.2 0.1>
   }
   skydive_GoingUp {
    angularVelocityScale=0.1
    fictitiousAngularScale=0
    jointName=root
    dragCoefficient=<0.01 0.01>
    gravityOverride=<0 -2000 400>
    liftCoefficient=<0.05 0.05>
    linearVelocityScale=<0.1 0.1 0.01>
   }
   skydive_Parachuting {
    angularVelocityScale=0.6
    fictitiousAngularScale=0.1
    jointName=root
    dragCoefficient=<0.8 0.8>
    gravityOverride=<0 -600 -500>
    liftCoefficient=<0.8 0.8>
    linearVelocityScale=<0.2 0.2 0.3>
   }
   skydive_Surfing {
    angularVelocityScale=0.6
    fictitiousAngularScale=0.1
    jointName=root
    dragCoefficient=<1.5 1.5>
    gravityOverride=<0 -600 -500>
    liftCoefficient=<1 1>
    linearVelocityScale=<0.1 0.1 0.3>
   }
   swimming {
    angularVelocityScale=0.6
    fictitiousAngularScale=0
    jointName=root
    dragCoefficient=<2.5 2.5>
    gravityOverride=<0 0 0>
    liftCoefficient=<0.1 0.1>
    linearVelocityScale=<0.6 0.6 0.5>
   }
   tactical_Sprint {
    angularVelocityScale=0.6
    fictitiousAngularScale=0.1
    jointName=root
    dragCoefficient=<2 2>
    gravityOverride=<0 0 -980>
    liftCoefficient=<0.5 0.5>
    linearVelocityScale=<0.6 0.6 0.5>
   }
   ziplining {
    angularVelocityScale=0.6
    fictitiousAngularScale=0
    jointName=root
    dragCoefficient=<0 1.5>
    gravityOverride=<0 0 -980>
    liftCoefficient=<0 0.3>
    linearVelocityScale=<0.4 0.4 0.5>
   }
  }
 }
 Gravity Override Params {type=GravityOverrideParamsStruct, default {skydive_FloatingDown {gravityOverride=<0 -700 900>}}}
 GravityOverride {type=Vector, default=<0 0 -980>}
 in vehicle rigid alpha {type=number}
 in vehicle scale alpha {type=number}
 RBAN Params {
  type=RBANParamsStruct
  default {
   dBNO {simSpaceSettings {
     maxAngularAcceleration=1000
     maxAngularVelocity=3
     maxLinearAcceleration=300
     maxLinearVelocity=700
     velocityScaleZ=0
     externalLinearDragV=<0 0.5 0>
    }}
   dBNO_Carried {simSpaceSettings {
     maxAngularAcceleration=1000
     maxAngularVelocity=3
     maxLinearAcceleration=300
     maxLinearVelocity=700
     velocityScaleZ=0
     externalLinearDragV=<0 0.5 0>
    }}
   emoteOrMelee {simSpaceSettings {
     maxAngularAcceleration=1000
     maxAngularVelocity=3
     maxLinearAcceleration=300
     maxLinearVelocity=700
     velocityScaleZ=0
     externalLinearDragV=<0 0.5 0>
    }}
   front_End {simSpaceSettings {
     maxAngularAcceleration=1000
     maxAngularVelocity=3
     maxLinearAcceleration=300
     maxLinearVelocity=700
     velocityScaleZ=0
     worldAlpha=0.3
     externalLinearDragV=<0 0.5 0>
    }}
   onGround_Crouch_Moving {simSpaceSettings {
     maxAngularAcceleration=1000
     maxAngularVelocity=3
     maxLinearAcceleration=300
     maxLinearVelocity=700
     velocityScaleZ=0
     externalLinearDragV=<0 0.5 0>
    }}
   onGround_Falling {simSpaceSettings {
     maxAngularAcceleration=1000
     maxAngularVelocity=3
     maxLinearAcceleration=300
     maxLinearVelocity=700
     velocityScaleZ=0
     externalLinearDragV=<0 0.5 0>
    }}
   onGround_Moving {simSpaceSettings {
     maxAngularAcceleration=1000
     maxAngularVelocity=3
     maxLinearAcceleration=300
     maxLinearVelocity=700
     velocityScaleZ=0
     externalLinearDragV=<0 0.5 0>
    }}
   onGround_Standing {simSpaceSettings {
     maxAngularAcceleration=1000
     maxAngularVelocity=3
     maxLinearAcceleration=300
     maxLinearVelocity=700
     velocityScaleZ=0
     externalLinearDragV=<0 0.5 0>
    }}
   rail_Grinding {simSpaceSettings {
     maxAngularAcceleration=1000
     maxAngularVelocity=3
     maxLinearAcceleration=300
     maxLinearVelocity=700
     velocityScaleZ=0
     worldAlpha=0.1
     externalLinearDragV=<0 0.5 0>
    }}
   riding_Moving {simSpaceSettings {
     maxAngularAcceleration=1000
     maxAngularVelocity=3
     maxLinearAcceleration=300
     maxLinearVelocity=700
     velocityScaleZ=0
     externalLinearDragV=<0 0.5 0>
    }}
   riding_Moving_Targeting {simSpaceSettings {
     maxAngularAcceleration=1000
     maxAngularVelocity=3
     maxLinearAcceleration=300
     maxLinearVelocity=700
     velocityScaleZ=0
     externalLinearDragV=<0 0.5 0>
    }}
   riding_Standing {simSpaceSettings {
     maxAngularAcceleration=1000
     maxAngularVelocity=3
     maxLinearAcceleration=300
     maxLinearVelocity=700
     velocityScaleZ=0
     externalLinearDragV=<0 0.5 0>
    }}
   skydive_DivingDown {simSpaceSettings {
     maxAngularAcceleration=1000
     maxAngularVelocity=3
     maxLinearAcceleration=300
     maxLinearVelocity=700
     velocityScaleZ=0
     externalLinearDragV=<0 0.5 0>
    }}
   skydive_FloatingDown {
    componentLinearAccScale=<0 0.5 0>
    componentLinearVelScale=<0 0.5 0>
    simSpaceSettings {
     maxAngularAcceleration=1000
     maxAngularVelocity=3
     maxLinearAcceleration=300
     maxLinearVelocity=700
     velocityScaleZ=0
     externalLinearDragV=<0 0.5 0>
    }
   }
   skydive_GoingUp {simSpaceSettings {
     maxAngularAcceleration=1000
     maxAngularVelocity=3
     maxLinearAcceleration=300
     maxLinearVelocity=700
     velocityScaleZ=0
     externalLinearDragV=<0 0.5 0>
    }}
   skydive_Parachuting {
    componentLinearAccScale=<0 0.1 0.3>
    componentLinearVelScale=<0 0.1 0.3>
    simSpaceSettings {
     maxAngularAcceleration=1000
     maxAngularVelocity=3
     maxLinearAcceleration=300
     maxLinearVelocity=700
     velocityScaleZ=0
     externalLinearDragV=<0 0.5 0>
    }
   }
   skydive_Surfing {simSpaceSettings {
     maxAngularAcceleration=1000
     maxAngularVelocity=3
     maxLinearAcceleration=300
     maxLinearVelocity=700
     velocityScaleZ=0
     externalLinearDragV=<0 0.5 0>
    }}
   swimming {simSpaceSettings {
     maxAngularAcceleration=1000
     maxAngularVelocity=3
     maxLinearAcceleration=300
     maxLinearVelocity=700
     velocityScaleZ=0
     externalLinearDragV=<0 0.5 0>
    }}
   tactical_Sprint {simSpaceSettings {
     maxAngularAcceleration=1000
     maxAngularVelocity=3
     maxLinearAcceleration=300
     maxLinearVelocity=700
     velocityScaleZ=0
     externalLinearDragV=<0 0.5 0>
    }}
   ziplining {simSpaceSettings {
     maxAngularAcceleration=1000
     maxAngularVelocity=3
     maxLinearAcceleration=300
     maxLinearVelocity=700
     velocityScaleZ=0
     externalLinearDragV=<0 0.5 0>
    }}
  }
 }
 RBAN Settings {
  type=RBANSettingStruct
  default {
   componentLinearAccScale=<0 0 0>
   componentLinearVelScale=<0 0 0>
   simSpaceSettings {
    dampingAlpha=1
    maxAngularAcceleration=1000
    maxAngularVelocity=3
    maxLinearAcceleration=300
    maxLinearVelocity=700
    velocityScaleZ=0
    worldAlpha=1
    externalAngularVelocity=<0 0 0>
    externalLinearDragV=<0 0.5 0>
    externalLinearVelocity=<0 0 0>
   }
  }
 }
]
