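# Physics asset holding cloth collision shapes for the medium female player body F_MED_Scribble, as its name indicates.
# One capsule covers the pelvis, and each upper arm has two nearly overlapping capsules, mirrored between left and right.
# Thigh and calf bodies are present but list no shapes.
# Constraints link calves to thighs, and thighs and upper arms to the pelvis, with all swing and twist motion free.
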PhysicsAsset /BRCosmetics/Characters/Player/Female/Medium/Bodies/F_MED_Scribble/Meshes/F_MED_Scribble_Physics_ClothColliders
bodies {
 calf_l []
 calf_r []
 pelvis [{shapeType=Capsule, length=5.8745, radius=9.8744, shapeName=upperarm_l_capsule, center=<-2.0677 -1.576 0>}]
 thigh_l []
 thigh_r []
 upperarm_l [
  {
   shapeType=Capsule
   length=22.5569
   radius=3.294
   shapeName=upperarm_l_capsule
   center=<12.7695 -1.0337 -0.7425>
   rotation=<rot 89.6019 -179.9775 -179.9777>
  }
  {
   shapeType=Capsule
   length=22.5569
   radius=3.294
   shapeName=upperarm_l_capsule
   center=<12.861 0.1989 -0.6841>
   rotation=<rot 87.6145 179.0031 -179.7919>
  }
 ]
 upperarm_r [
  {
   shapeType=Capsule
   length=22.5569
   radius=3.294
   shapeName=upperarm_l_capsule
   center=<-12.7695 1.0337 0.7425>
   rotation=<rot 89.6019 -179.9775 0.0223>
  }
  {
   shapeType=Capsule
   length=22.5569
   radius=3.294
   shapeName=upperarm_l_capsule
   center=<-12.861 -0.1989 0.6841>
   rotation=<rot 87.6145 179.0031 0.2081>
  }
 ]
}
constraints [
 {
  bone1Name=calf_l
  bone2Name=thigh_l
  swing1LimitDegrees=45
  swing1Motion=Free
  swing2LimitDegrees=45
  swing2Motion=Free
  twistLimitDegrees=45
  twistMotion=Free
 }
 {
  bone1Name=upperarm_l
  bone2Name=pelvis
  swing1LimitDegrees=45
  swing1Motion=Free
  swing2LimitDegrees=45
  swing2Motion=Free
  twistLimitDegrees=45
  twistMotion=Free
 }
 {
  bone1Name=thigh_l
  bone2Name=pelvis
  swing1LimitDegrees=45
  swing1Motion=Free
  swing2LimitDegrees=45
  swing2Motion=Free
  twistLimitDegrees=45
  twistMotion=Free
 }
 {
  bone1Name=upperarm_r
  bone2Name=pelvis
  swing1LimitDegrees=45
  swing1Motion=Free
  swing2LimitDegrees=45
  swing2Motion=Free
  twistLimitDegrees=45
  twistMotion=Free
 }
 {
  bone1Name=thigh_r
  bone2Name=pelvis
  swing1LimitDegrees=45
  swing1Motion=Free
  swing2LimitDegrees=45
  swing2Motion=Free
  twistLimitDegrees=45
  twistMotion=Free
 }
 {
  bone1Name=calf_r
  bone2Name=thigh_r
  swing1LimitDegrees=45
  swing1Motion=Free
  swing2LimitDegrees=45
  swing2Motion=Free
  twistLimitDegrees=45
  twistMotion=Free
 }
]
