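# Physics asset for the dynamic hair on the Scribble female medium body's face accessory part.
# Five three-capsule chains (back, back left and right, front left and right) hang from the head with tight twist and limited swing.
# Head, neck, upper spine and clavicle capsules form the body volume, joined by angularly free constraints.
# Extra angularly free constraints tie the middle segments of both side-back chains to the center-back chain.
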PhysicsAsset /BRCosmetics/Characters/Player/Female/Medium/Bodies/F_MED_Scribble/Meshes/Parts/F_MED_Scribble_FaceAcc_Dyn_Hair
bodies {
 clavicle_l [{
   shapeType=Capsule
   length=10.8197
   radius=6.9939
   shapeName=head_capsule
   center=<3.4284 -0.658 -1.862>
   rotation=<rot 80.0643 -0.0007 -2.2482>
  }]
 clavicle_r [{
   shapeType=Capsule
   length=10.8197
   radius=6.9939
   shapeName=head_capsule
   center=<-3.4284 0.658 1.862>
   rotation=<rot 80.0643 -0.0007 177.7518>
  }]
 dyn_hair_bk_01 [{shapeType=Capsule, length=4.2921, radius=3.5, shapeName=bunny_ear_l_01_dyn_capsule, center=<2.5 0 3.5>, rotation=<rot 90 0 0>}]
 dyn_hair_bk_02 [{shapeType=Capsule, length=4.2921, radius=3.5, shapeName=bunny_ear_l_01_dyn_capsule, center=<2.5 0 3.5>, rotation=<rot 90 0 0>}]
 dyn_hair_bk_03 [{shapeType=Capsule, length=4.2921, radius=3.5, shapeName=bunny_ear_l_01_dyn_capsule, center=<2.5 0 3.5>, rotation=<rot 90 0 0>}]
 dyn_hair_bk_l_01 [{shapeType=Capsule, length=4.2921, radius=3.5, shapeName=bunny_ear_l_01_dyn_capsule, center=<2.5 0 2.4805>, rotation=<rot 90 0 0>}]
 dyn_hair_bk_l_02 [{shapeType=Capsule, length=4.2921, radius=3.5, shapeName=bunny_ear_l_01_dyn_capsule, center=<2.5 0 2.3411>, rotation=<rot 90 0 0>}]
 dyn_hair_bk_l_03 [{shapeType=Capsule, length=4.2921, radius=3.5, shapeName=bunny_ear_l_01_dyn_capsule, center=<2.5 0 2.0954>, rotation=<rot 90 0 0>}]
 dyn_hair_bk_r_01 [{shapeType=Capsule, length=4.2921, radius=3.5, shapeName=bunny_ear_l_01_dyn_capsule, center=<2.5 0 2.4805>, rotation=<rot 90 0 0>}]
 dyn_hair_bk_r_02 [{shapeType=Capsule, length=4.2921, radius=3.5, shapeName=bunny_ear_l_01_dyn_capsule, center=<2.5 0 2.3411>, rotation=<rot 90 0 0>}]
 dyn_hair_bk_r_03 [{shapeType=Capsule, length=4.2921, radius=3.5, shapeName=bunny_ear_l_01_dyn_capsule, center=<2.5 0 2.0954>, rotation=<rot 90 0 0>}]
 dyn_hair_fr_l_01 [{shapeType=Capsule, length=4.2921, radius=3.5, shapeName=bunny_ear_l_01_dyn_capsule, center=<2.5 0 3.5632>, rotation=<rot 90 0 0>}]
 dyn_hair_fr_l_02 [{shapeType=Capsule, length=4.2921, radius=3.5, shapeName=bunny_ear_l_01_dyn_capsule, center=<2.5 0 3.5103>, rotation=<rot 90 0 0>}]
 dyn_hair_fr_l_03 [{shapeType=Capsule, length=4.2921, radius=3.5, shapeName=bunny_ear_l_01_dyn_capsule, center=<2.5 0 3.8429>, rotation=<rot 90 0 0>}]
 dyn_hair_fr_r_01 [{shapeType=Capsule, length=4.2921, radius=3.5, shapeName=bunny_ear_l_01_dyn_capsule, center=<2.5 0 3.5632>, rotation=<rot 90 0 0>}]
 dyn_hair_fr_r_02 [{shapeType=Capsule, length=4.2921, radius=3.5, shapeName=bunny_ear_l_01_dyn_capsule, center=<2.5 0 3.5103>, rotation=<rot 90 0 0>}]
 dyn_hair_fr_r_03 [{shapeType=Capsule, length=4.2921, radius=3.5, shapeName=bunny_ear_l_01_dyn_capsule, center=<2.5 0 3.8429>, rotation=<rot 90 0 0>}]
 head [{shapeType=Capsule, length=10.1443, radius=7.6694, shapeName=head_capsule, center=<-2.2033 0.5837 0>, rotation=<rot 90 0 0>}]
 neck_02 [{shapeType=Capsule, length=6.9747, radius=4.4339, shapeName=head_capsule, center=<-2.3725 0.3118 0>, rotation=<rot 90 178.2101 168.5099>}]
 spine_05 [{shapeType=Capsule, length=12.698, radius=7.4151, shapeName=head_capsule, center=<2.963 -1.0152 0>}]
}
constraints [
 {bone1Name=dyn_hair_bk_02, bone2Name=dyn_hair_bk_01, swing1LimitDegrees=15, swing2LimitDegrees=45, twistLimitDegrees=5}
 {bone1Name=dyn_hair_bk_03, bone2Name=dyn_hair_bk_02, swing1LimitDegrees=15, swing2LimitDegrees=45, twistLimitDegrees=5}
 {bone1Name=dyn_hair_bk_l_02, bone2Name=dyn_hair_bk_l_01, swing1LimitDegrees=15, swing2LimitDegrees=45, twistLimitDegrees=5}
 {bone1Name=dyn_hair_bk_l_03, bone2Name=dyn_hair_bk_l_02, swing1LimitDegrees=15, swing2LimitDegrees=45, twistLimitDegrees=5}
 {bone1Name=dyn_hair_bk_r_02, bone2Name=dyn_hair_bk_r_01, swing1LimitDegrees=15, swing2LimitDegrees=45, twistLimitDegrees=5}
 {bone1Name=dyn_hair_bk_r_03, bone2Name=dyn_hair_bk_r_02, swing1LimitDegrees=15, swing2LimitDegrees=45, twistLimitDegrees=5}
 {bone1Name=dyn_hair_fr_l_02, bone2Name=dyn_hair_fr_l_01, swing1LimitDegrees=15, swing2LimitDegrees=45, twistLimitDegrees=5}
 {bone1Name=dyn_hair_fr_l_03, bone2Name=dyn_hair_fr_l_02, swing1LimitDegrees=15, swing2LimitDegrees=45, twistLimitDegrees=5}
 {bone1Name=dyn_hair_fr_r_02, bone2Name=dyn_hair_fr_r_01, swing1LimitDegrees=15, swing2LimitDegrees=45, twistLimitDegrees=5}
 {bone1Name=dyn_hair_fr_r_03, bone2Name=dyn_hair_fr_r_02, swing1LimitDegrees=15, swing2LimitDegrees=45, twistLimitDegrees=5}
 {bone1Name=dyn_hair_bk_01, bone2Name=head, swing1LimitDegrees=15, swing2LimitDegrees=45, twistLimitDegrees=5}
 {bone1Name=dyn_hair_bk_l_01, bone2Name=head, swing1LimitDegrees=15, swing2LimitDegrees=45, twistLimitDegrees=5}
 {bone1Name=dyn_hair_fr_l_01, bone2Name=head, swing1LimitDegrees=15, swing2LimitDegrees=45, twistLimitDegrees=5}
 {bone1Name=dyn_hair_bk_r_01, bone2Name=head, swing1LimitDegrees=15, swing2LimitDegrees=45, twistLimitDegrees=5}
 {bone1Name=dyn_hair_fr_r_01, bone2Name=head, swing1LimitDegrees=15, swing2LimitDegrees=45, twistLimitDegrees=5}
 {
  bone1Name=head
  bone2Name=neck_02
  swing1LimitDegrees=45
  swing1Motion=Free
  swing2LimitDegrees=45
  swing2Motion=Free
  twistLimitDegrees=45
  twistMotion=Free
 }
 {
  bone1Name=neck_02
  bone2Name=spine_05
  swing1LimitDegrees=45
  swing1Motion=Free
  swing2LimitDegrees=45
  swing2Motion=Free
  twistLimitDegrees=45
  twistMotion=Free
 }
 {
  bone1Name=clavicle_l
  bone2Name=spine_05
  swing1LimitDegrees=45
  swing1Motion=Free
  swing2LimitDegrees=45
  swing2Motion=Free
  twistLimitDegrees=45
  twistMotion=Free
 }
 {
  bone1Name=clavicle_r
  bone2Name=spine_05
  swing1LimitDegrees=45
  swing1Motion=Free
  swing2LimitDegrees=45
  swing2Motion=Free
  twistLimitDegrees=45
  twistMotion=Free
 }
 {
  bone1Name=dyn_hair_bk_l_02
  bone2Name=dyn_hair_bk_02
  swing1LimitDegrees=45
  swing1Motion=Free
  swing2LimitDegrees=45
  swing2Motion=Free
  twistLimitDegrees=45
  twistMotion=Free
 }
 {
  bone1Name=dyn_hair_bk_r_02
  bone2Name=dyn_hair_bk_02
  swing1LimitDegrees=45
  swing1Motion=Free
  swing2LimitDegrees=45
  swing2Motion=Free
  twistLimitDegrees=45
  twistMotion=Free
 }
]
