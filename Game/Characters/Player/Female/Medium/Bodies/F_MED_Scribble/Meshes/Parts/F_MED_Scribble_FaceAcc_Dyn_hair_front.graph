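# Physics asset for the front hair of the Scribble female medium player character, covering three bang strands: center, mid-left and mid-right.
# Each strand is a two-bone chain of capsules, with the root bone constrained to a head capsule and the second bone to the first.
# All joints share the same angular limits, with tight twist and uneven swing.
# The bang capsules are identical and keep a bunny-ear shape name.
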PhysicsAsset /Game/Characters/Player/Female/Medium/Bodies/F_MED_Scribble/Meshes/Parts/F_MED_Scribble_FaceAcc_Dyn_hair_front
bodies {
 dyn_bangs_01 [{shapeType=Capsule, length=4.2921, radius=3.5, shapeName=bunny_ear_l_01_dyn_capsule, center=<2.5 0 2.8>, rotation=<rot 90 0 0>}]
 dyn_bangs_02 [{shapeType=Capsule, length=4.2921, radius=3.5, shapeName=bunny_ear_l_01_dyn_capsule, center=<2.5 0 2.8>, rotation=<rot 90 0 0>}]
 dyn_bangs_mid_l_01 [{shapeType=Capsule, length=4.2921, radius=3.5, shapeName=bunny_ear_l_01_dyn_capsule, center=<2.5 0 2.8>, rotation=<rot 90 0 0>}]
 dyn_bangs_mid_l_02 [{shapeType=Capsule, length=4.2921, radius=3.5, shapeName=bunny_ear_l_01_dyn_capsule, center=<2.5 0 2.8>, rotation=<rot 90 0 0>}]
 dyn_bangs_mid_r_01 [{shapeType=Capsule, length=4.2921, radius=3.5, shapeName=bunny_ear_l_01_dyn_capsule, center=<2.5 0 2.8>, rotation=<rot 90 0 0>}]
 dyn_bangs_mid_r_02 [{shapeType=Capsule, length=4.2921, radius=3.5, shapeName=bunny_ear_l_01_dyn_capsule, center=<2.5 0 2.8>, rotation=<rot 90 0 0>}]
 head [{shapeType=Capsule, length=10.1443, radius=7.6694, shapeName=head_capsule, center=<-2.2033 0.5837 0>, rotation=<rot 90 0 0>}]
}
constraints [
 {bone1Name=dyn_bangs_02, bone2Name=dyn_bangs_01, swing1LimitDegrees=15, swing2LimitDegrees=45, twistLimitDegrees=5}
 {bone1Name=dyn_bangs_mid_l_02, bone2Name=dyn_bangs_mid_l_01, swing1LimitDegrees=15, swing2LimitDegrees=45, twistLimitDegrees=5}
 {bone1Name=dyn_bangs_mid_r_02, bone2Name=dyn_bangs_mid_r_01, swing1LimitDegrees=15, swing2LimitDegrees=45, twistLimitDegrees=5}
 {bone1Name=dyn_bangs_mid_l_01, bone2Name=head, swing1LimitDegrees=15, swing2LimitDegrees=45, twistLimitDegrees=5}
 {bone1Name=dyn_bangs_01, bone2Name=head, swing1LimitDegrees=15, swing2LimitDegrees=45, twistLimitDegrees=5}
 {bone1Name=dyn_bangs_mid_r_01, bone2Name=head, swing1LimitDegrees=15, swing2LimitDegrees=45, twistLimitDegrees=5}
]
