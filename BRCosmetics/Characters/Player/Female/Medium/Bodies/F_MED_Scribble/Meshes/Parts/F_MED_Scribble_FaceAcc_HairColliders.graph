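# Physics asset of hair colliders for the face accessory part of the F_MED_Scribble female player character, giving hair a simplified upper body to collide against.
# Single capsules cover the head, both neck bones, both clavicles and both upper arms, with left and right sides mirrored.
# Constraints link head to neck and each upper arm to its clavicle in three separate chains, with swing and twist left free.
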PhysicsAsset /BRCosmetics/Characters/Player/Female/Medium/Bodies/F_MED_Scribble/Meshes/Parts/F_MED_Scribble_FaceAcc_HairColliders
bodies {
 clavicle_l [{shapeType=Capsule, length=10.1422, radius=5.9282, shapeName=head_capsule, center=<3.7331 -1.2869 -0.5831>, rotation=<rot 80 9.9987 4.1008>}]
 clavicle_r [{shapeType=Capsule, length=10.1422, radius=5.9282, shapeName=head_capsule, center=<-3.7331 1.2869 0.5831>, rotation=<rot 80 9.9987 -175.8992>}]
 head [{shapeType=Capsule, length=6.7222, radius=6.5212, shapeName=head_capsule, center=<1.2017 2.5372 0>, rotation=<rot 90 0 0>}]
 neck_01 [{shapeType=Capsule, length=3.7834, radius=5.2373, shapeName=head_capsule, center=<-0.7662 0.4357 0>, rotation=<rot 90 0 -9.9998>}]
 neck_02 [{shapeType=Capsule, length=3.9629, radius=5.0578, shapeName=head_capsule, center=<-0.7662 0.4357 0>, rotation=<rot 90 0 -9.9998>}]
 upperarm_l [{
   shapeType=Capsule
   length=19.0498
   radius=5.1226
   shapeName=head_capsule
   center=<9.8038 0 -0.0475>
   rotation=<rot 87.4976 -179.9975 -179.9976>
  }]
 upperarm_r [{shapeType=Capsule, length=19.0498, radius=5.1226, shapeName=head_capsule, center=<-9.8038 0 0.0475>, rotation=<rot 87.4976 -179.9975 0.0024>}]
}
constraints [
 {
  bone1Name=head
  bone2Name=neck_02
  swing1LimitDegrees=45
  swing1Motion=Free
  swing2LimitDegrees=45
  swing2Motion=Free
  twistLimitDegrees=45
  twistMotion=Free
 }
 {
  bone1Name=neck_02
  bone2Name=neck_01
  swing1LimitDegrees=45
  swing1Motion=Free
  swing2LimitDegrees=45
  swing2Motion=Free
  twistLimitDegrees=45
  twistMotion=Free
 }
 {
  bone1Name=upperarm_l
  bone2Name=clavicle_l
  swing1LimitDegrees=45
  swing1Motion=Free
  swing2LimitDegrees=45
  swing2Motion=Free
  twistLimitDegrees=45
  twistMotion=Free
 }
 {
  bone1Name=upperarm_r
  bone2Name=clavicle_r
  swing1LimitDegrees=45
  swing1Motion=Free
  swing2LimitDegrees=45
  swing2Motion=Free
  twistLimitDegrees=45
  twistMotion=Free
 }
]
